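# Physics asset for the PeonyBellow female medium body cosmetic, covering its jacket flaps and front and back key chains.
# Capsules on pelvis, spine_01 and thighs form the body, with small base jacket anchors linked to spine_01 by free-rotation constraints.
# Each dynamic flap has a single-axis swing limit from its anchor, tightest at the back and widest at the sides.
# Both two-segment key chains hang from spine_01 with unrestricted rotation.
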PhysicsAsset /BRCosmetics/Characters/Player/Female/Medium/Bodies/F_MED_PeonyBellow/Meshes/F_MED_PeonyBellow_Physics
bodies {
 base_jacket_bk [{shapeType=Capsule, length=2, radius=0.5, shapeName=base_jacket_bk_capsule}]
 base_jacket_bk_l [{shapeType=Capsule, length=2, radius=0.5, shapeName=base_jacket_bk_l_capsule}]
 base_jacket_bk_r [{shapeType=Capsule, length=2, radius=0.5, shapeName=base_jacket_bk_r_capsule}]
 base_jacket_fr_l [{shapeType=Capsule, length=2, radius=0.5, shapeName=base_jacket_fr_l_capsule}]
 base_jacket_fr_r [{shapeType=Capsule, length=2, radius=0.5, shapeName=base_jacket_fr_r_capsule}]
 base_jacket_l [{shapeType=Capsule, length=2, radius=0.5, shapeName=base_jacket_l_capsule}]
 base_jacket_r [{shapeType=Capsule, length=2, radius=0.5, shapeName=base_jacket_r_capsule}]
 dyn_jacket_bk [{shapeType=Capsule, length=4.8403, radius=4.5, shapeName=dyn_jacket_bk_capsule, center=<3.8068 -1.9596 0>, rotation=<rot 90 0 0>}]
 dyn_jacket_bk_l [{shapeType=Capsule, length=5.9403, radius=4.5, shapeName=dyn_jacket_bk_l_capsule, center=<4.3476 -1.9588 0>, rotation=<rot 90 0 0>}]
 dyn_jacket_bk_r [{shapeType=Capsule, length=5.9403, radius=4.5, shapeName=dyn_jacket_bk_r_capsule, center=<4.3476 -1.9588 0>, rotation=<rot 90 0 180>}]
 dyn_jacket_fr_l [{
   shapeType=Capsule
   length=9.003
   radius=2.8798
   shapeName=dyn_jacket_fr_l_capsule
   center=<4.1277 -0.7192 1.6023>
   rotation=<rot -86.0922 75.45 -80.6329>
  }]
 dyn_jacket_fr_r [{
   shapeType=Capsule
   length=7.9788
   radius=3.904
   shapeName=dyn_jacket_fr_r_capsule
   center=<4.081 -1.3352 -2.4465>
   rotation=<rot -86.0922 75.45 99.3671>
  }]
 dyn_jacket_l [{shapeType=Capsule, length=7, radius=4.5, shapeName=dyn_jacket_l_capsule, center=<4.5792 -1.7242 -0.0161>, rotation=<rot 87.809 -180 180>}]
 dyn_jacket_r [{shapeType=Capsule, length=7, radius=4.5, shapeName=dyn_jacket_r_capsule, center=<4.5792 -1.7242 0.0161>, rotation=<rot 87.809 -180 0>}]
 dyn_keyChain_b_1 [{
   shapeType=Capsule
   length=8.2137
   radius=1.2582
   shapeName=dyn_keyChain_b_1_capsule
   center=<6.0481 -0.787 0.0152>
   rotation=<rot 0 90.3088 -89.9218>
  }]
 dyn_keyChain_b_2 [{
   shapeType=Capsule
   length=5.4051
   radius=1.4173
   shapeName=dyn_keyChain_b_2_capsule
   center=<2.5779 -0.9556 0>
   rotation=<rot 90 180 -174.4117>
  }]
 dyn_keyChain_f_1 [{
   shapeType=Capsule
   length=5.7682
   radius=1.433
   shapeName=dyn_keyChain_f_1_capsule
   center=<4.7545 -0.0044 0.0846>
   rotation=<rot -90 179.5524 -177.0737>
  }]
 dyn_keyChain_f_2 [{
   shapeType=Capsule
   length=5.7256
   radius=2.7622
   shapeName=dyn_keyChain_f_2_capsule
   center=<3.6466 0.6352 0.1264>
   rotation=<rot -7.2422 -90.3877 -4.7399>
  }]
 pelvis [{shapeType=Capsule, length=8.6013, radius=11.1086, shapeName=pelvis_capsule, center=<-3.0781 -1.5342 0>}]
 spine_01 [{shapeType=Capsule, length=5.3637, radius=8.7916, shapeName=spine_01_capsule, center=<0 0.4982 0>}]
 thigh_l [{shapeType=Capsule, length=24.1089, radius=8.3516, shapeName=thigh_l_capsule, center=<-16.0429 -0.1858 0>, rotation=<rot 90 9.7524 12.4986>}]
 thigh_r [{shapeType=Capsule, length=24.1089, radius=8.3516, shapeName=thigh_r_capsule, center=<16.0429 0.1858 0>, rotation=<rot 90 8.8807 -168.3731>}]
}
constraints [
 {
  bone1Name=spine_01
  bone2Name=pelvis
  swing1LimitDegrees=45
  swing1Motion=Free
  swing2LimitDegrees=45
  swing2Motion=Free
  twistLimitDegrees=45
  twistMotion=Free
 }
 {
  bone1Name=thigh_r
  bone2Name=pelvis
  swing1LimitDegrees=45
  swing1Motion=Free
  swing2LimitDegrees=45
  swing2Motion=Free
  twistLimitDegrees=45
  twistMotion=Free
 }
 {
  bone1Name=thigh_l
  bone2Name=pelvis
  swing1LimitDegrees=45
  swing1Motion=Free
  swing2LimitDegrees=45
  swing2Motion=Free
  twistLimitDegrees=45
  twistMotion=Free
 }
 {
  bone1Name=base_jacket_bk
  bone2Name=spine_01
  swing1LimitDegrees=45
  swing1Motion=Free
  swing2LimitDegrees=45
  swing2Motion=Free
  twistLimitDegrees=45
  twistMotion=Free
 }
 {bone1Name=dyn_jacket_bk, bone2Name=base_jacket_bk, swing1LimitDegrees=10, swing2LimitDegrees=0, twistLimitDegrees=0}
 {
  bone1Name=base_jacket_bk_l
  bone2Name=spine_01
  swing1LimitDegrees=45
  swing1Motion=Free
  swing2LimitDegrees=45
  swing2Motion=Free
  twistLimitDegrees=45
  twistMotion=Free
 }
 {bone1Name=dyn_jacket_bk_l, bone2Name=base_jacket_bk_l, swing1LimitDegrees=10, swing2LimitDegrees=0, twistLimitDegrees=0}
 {
  bone1Name=base_jacket_bk_r
  bone2Name=spine_01
  swing1LimitDegrees=45
  swing1Motion=Free
  swing2LimitDegrees=45
  swing2Motion=Free
  twistLimitDegrees=45
  twistMotion=Free
 }
 {bone1Name=dyn_jacket_bk_r, bone2Name=base_jacket_bk_r, swing1LimitDegrees=10, swing2LimitDegrees=0, twistLimitDegrees=0}
 {
  bone1Name=base_jacket_fr_l
  bone2Name=spine_01
  swing1LimitDegrees=45
  swing1Motion=Free
  swing2LimitDegrees=45
  swing2Motion=Free
  twistLimitDegrees=45
  twistMotion=Free
 }
 {bone1Name=dyn_jacket_fr_l, bone2Name=base_jacket_fr_l, swing1LimitDegrees=25, swing2LimitDegrees=0, twistLimitDegrees=0}
 {
  bone1Name=base_jacket_fr_r
  bone2Name=spine_01
  swing1LimitDegrees=45
  swing1Motion=Free
  swing2LimitDegrees=45
  swing2Motion=Free
  twistLimitDegrees=45
  twistMotion=Free
 }
 {bone1Name=dyn_jacket_fr_r, bone2Name=base_jacket_fr_r, swing1LimitDegrees=25, swing2LimitDegrees=0, twistLimitDegrees=0}
 {
  bone1Name=base_jacket_l
  bone2Name=spine_01
  swing1LimitDegrees=45
  swing1Motion=Free
  swing2LimitDegrees=45
  swing2Motion=Free
  twistLimitDegrees=45
  twistMotion=Free
 }
 {bone1Name=dyn_jacket_l, bone2Name=base_jacket_l, swing1LimitDegrees=30, swing2LimitDegrees=0, twistLimitDegrees=0}
 {
  bone1Name=base_jacket_r
  bone2Name=spine_01
  swing1LimitDegrees=45
  swing1Motion=Free
  swing2LimitDegrees=45
  swing2Motion=Free
  twistLimitDegrees=45
  twistMotion=Free
 }
 {bone1Name=dyn_jacket_r, bone2Name=base_jacket_r, swing1LimitDegrees=30, swing2LimitDegrees=0, twistLimitDegrees=0}
 {
  bone1Name=dyn_keyChain_b_1
  bone2Name=spine_01
  swing1LimitDegrees=45
  swing1Motion=Free
  swing2LimitDegrees=45
  swing2Motion=Free
  twistLimitDegrees=45
  twistMotion=Free
 }
 {
  bone1Name=dyn_keyChain_b_2
  bone2Name=dyn_keyChain_b_1
  swing1LimitDegrees=45
  swing1Motion=Free
  swing2LimitDegrees=45
  swing2Motion=Free
  twistLimitDegrees=45
  twistMotion=Free
 }
 {
  bone1Name=dyn_keyChain_f_1
  bone2Name=spine_01
  swing1LimitDegrees=45
  swing1Motion=Free
  swing2LimitDegrees=45
  swing2Motion=Free
  twistLimitDegrees=45
  twistMotion=Free
 }
 {
  bone1Name=dyn_keyChain_f_2
  bone2Name=dyn_keyChain_f_1
  swing1LimitDegrees=45
  swing1Motion=Free
  swing2LimitDegrees=45
  swing2Motion=Free
  twistLimitDegrees=45
  twistMotion=Free
 }
]
