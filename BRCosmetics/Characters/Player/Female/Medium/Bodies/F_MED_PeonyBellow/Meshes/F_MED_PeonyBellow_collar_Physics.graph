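# Physics asset for the neck collar of the female medium PeonyBellow cosmetic body, defining secondary motion and collision shapes for the collar.
# Seven dynamic collar capsules (center plus three per side) each constrain to a tiny base-bone capsule, with narrow swing on one axis, wider on the other, and zero twist limit.
# Head, jaw, hair and two neck capsules supply body volume around the collar, linked by free-rotating head-to-neck constraints.
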PhysicsAsset /BRCosmetics/Characters/Player/Female/Medium/Bodies/F_MED_PeonyBellow/Meshes/F_MED_PeonyBellow_collar_Physics
bodies {
 base_neckCollar_c_1 [{shapeType=Capsule, length=1, radius=0.5, shapeName=base_neckCollar_c_1_capsule}]
 base_neckCollar_l_1 [{shapeType=Capsule, length=1, radius=0.5, shapeName=base_neckCollar_l_1_capsule}]
 base_neckCollar_l_2 [{shapeType=Capsule, length=1, radius=0.5, shapeName=base_neckCollar_l_2_capsule}]
 base_neckCollar_l_3 [{shapeType=Capsule, length=1, radius=0.5, shapeName=base_neckCollar_l_3_capsule}]
 base_neckCollar_r_1 [{shapeType=Capsule, length=1, radius=0.5, shapeName=base_neckCollar_r_1_capsule}]
 base_neckCollar_r_2 [{shapeType=Capsule, length=1, radius=0.5, shapeName=base_neckCollar_r_2_capsule}]
 base_neckCollar_r_3 [{shapeType=Capsule, length=1, radius=0.5, shapeName=base_neckCollar_r_3_capsule}]
 dyn_neckCollar_c_1 [{
   shapeType=Capsule
   length=3.4471
   radius=6.072
   shapeName=dyn_neckCollar_c_1_capsule
   center=<3.6518 -0.865 0.9093>
   rotation=<rot 0 -90 -82.5139>
  }]
 dyn_neckCollar_l_1 [{
   shapeType=Capsule
   length=5.8294
   radius=4.0727
   shapeName=dyn_neckCollar_l_1_capsule
   center=<1.9847 0.9294 0.9393>
   rotation=<rot -1.4462 -22.3408 104.6316>
  }]
 dyn_neckCollar_l_2 [{
   shapeType=Capsule
   length=2.265
   radius=5.893
   shapeName=dyn_neckCollar_l_2_capsule
   center=<3.9586 1.5644 0.9845>
   rotation=<rot 86.9583 37.5585 25.1815>
  }]
 dyn_neckCollar_l_3 [{
   shapeType=Capsule
   length=2.6903
   radius=5.9799
   shapeName=dyn_neckCollar_l_3_capsule
   center=<4.11 0.2727 0.3045>
   rotation=<rot 77.6706 17.1807 12.5719>
  }]
 dyn_neckCollar_r_1 [{
   shapeType=Capsule
   length=5.8294
   radius=4.0727
   shapeName=dyn_neckCollar_l_1_capsule
   center=<1.9847 0.9294 0.9393>
   rotation=<rot 1.4462 22.3408 -104.6316>
  }]
 dyn_neckCollar_r_2 [{
   shapeType=Capsule
   length=2.265
   radius=5.893
   shapeName=dyn_neckCollar_l_2_capsule
   center=<3.9586 -1.5644 0.9845>
   rotation=<rot 86.9583 37.5593 47.0395>
  }]
 dyn_neckCollar_r_3 [{
   shapeType=Capsule
   length=2.6903
   radius=5.9799
   shapeName=dyn_neckCollar_l_3_capsule
   center=<4.11 -0.2727 0.3045>
   rotation=<rot 77.6706 17.1807 12.5719>
  }]
 head [
  {shapeType=Capsule, length=4, radius=11, shapeName=head_capsule, center=<5.1868 -3.1582 0>, rotation=<rot -90 0 -9.3174>}
  {shapeType=Capsule, length=4.4356, radius=6.3426, shapeName=jaw_capsule, center=<0.9945 4.4127 0>, rotation=<rot -90 0 -9.3174>}
  {
   shapeType=Capsule
   length=9.5657
   radius=6.3426
   shapeName=hair_capsule
   center=<5.8824 -5.4508 -3.5085>
   rotation=<rot -64.15 55.5321 143.8486>
  }
 ]
 neck_01 [{shapeType=Capsule, length=3.0742, radius=4.9625, shapeName=neck_01_capsule, center=<0.2812 0 0>, rotation=<rot -90 0 0>}]
 neck_02 [{shapeType=Capsule, length=1.9794, radius=4.175, shapeName=neck_01_capsule, rotation=<rot -90 0 0>}]
}
constraints [
 {bone1Name=dyn_neckCollar_c_1, bone2Name=base_neckCollar_c_1, swing1LimitDegrees=4, swing2LimitDegrees=30, twistLimitDegrees=0}
 {bone1Name=dyn_neckCollar_l_1, bone2Name=base_neckCollar_l_1, swing1LimitDegrees=4, swing2LimitDegrees=30, twistLimitDegrees=0}
 {bone1Name=dyn_neckCollar_l_2, bone2Name=base_neckCollar_l_2, swing1LimitDegrees=4, swing2LimitDegrees=30, twistLimitDegrees=0}
 {bone1Name=dyn_neckCollar_l_3, bone2Name=base_neckCollar_l_3, swing1LimitDegrees=4, swing2LimitDegrees=30, twistLimitDegrees=0}
 {bone1Name=dyn_neckCollar_r_1, bone2Name=base_neckCollar_r_1, swing1LimitDegrees=4, swing2LimitDegrees=30, twistLimitDegrees=0}
 {bone1Name=dyn_neckCollar_r_2, bone2Name=base_neckCollar_r_2, swing1LimitDegrees=4, swing2LimitDegrees=30, twistLimitDegrees=0}
 {bone1Name=dyn_neckCollar_r_3, bone2Name=base_neckCollar_r_3, swing1LimitDegrees=4, swing2LimitDegrees=30, twistLimitDegrees=0}
 {
  bone1Name=head
  bone2Name=neck_02
  swing1LimitDegrees=45
  swing1Motion=Free
  swing2LimitDegrees=45
  swing2Motion=Free
  twistLimitDegrees=45
  twistMotion=Free
 }
 {
  bone1Name=neck_02
  bone2Name=neck_01
  swing1LimitDegrees=45
  swing1Motion=Free
  swing2LimitDegrees=45
  swing2Motion=Free
  twistLimitDegrees=45
  twistMotion=Free
 }
]
