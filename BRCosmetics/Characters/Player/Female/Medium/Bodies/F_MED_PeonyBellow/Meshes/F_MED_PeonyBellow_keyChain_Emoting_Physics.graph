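# Physics asset for the key chain on the PeonyBellow female medium character body, an emoting variant by its name.
# Back and front capsule chains run from base bodies linked to spine_01, with tight swing limits and zero twist.
# A fully locked constraint joins the two chain tips, closing the loop.
# Pelvis, spine and right-thigh capsules represent the body, and a huge zero-length capsule below the floor bone approximates the ground.
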PhysicsAsset /BRCosmetics/Characters/Player/Female/Medium/Bodies/F_MED_PeonyBellow/Meshes/F_MED_PeonyBellow_keyChain_Emoting_Physics
bodies {
 base_keyChain_b_1 [{shapeType=Capsule, length=1, radius=0.5, shapeName=base_keyChain_b_1_capsule}]
 base_keyChain_f_1 [{shapeType=Capsule, length=1, radius=0.5, shapeName=base_keyChain_f_1_capsule}]
 dyn_keyChain_b_1 [{
   shapeType=Capsule
   length=5.058
   radius=2.2034
   shapeName=dyn_keyChain_b_1_capsule
   center=<6.0518 -1.4714 0.0152>
   rotation=<rot 0 90.3088 -89.9218>
  }]
 dyn_keyChain_b_2 [{
   shapeType=Capsule
   length=5.1838
   radius=2.583
   shapeName=dyn_keyChain_b_2_capsule
   center=<2.5021 -1.7301 0>
   rotation=<rot 90 27.5984 33.1867>
  }]
 dyn_keyChain_b_3 [{
   shapeType=Capsule
   length=22.705
   radius=3.358
   shapeName=dyn_keyChain_b_3_capsule
   center=<1.7091 -0.5252 -8.1622>
   rotation=<rot -3.3356 -90 1.8937>
  }]
 dyn_keyChain_b_4 [{
   shapeType=Capsule
   length=22.7147
   radius=3.3664
   shapeName=dyn_keyChain_b_4_capsule
   center=<1.474 -0.7205 -8.1778>
   rotation=<rot -2.0766 -90 -0.2472>
  }]
 dyn_keyChain_b_5 [{
   shapeType=Capsule
   length=22.7205
   radius=3.4792
   shapeName=dyn_keyChain_b_5_capsule
   center=<1.7756 -0.9936 -8.203>
   rotation=<rot -0.2855 -90 0.7181>
  }]
 dyn_keyChain_b_6 [{
   shapeType=Capsule
   length=7.5875
   radius=3.2307
   shapeName=dyn_keyChain_b_6_capsule
   center=<5.3412 -1.9266 -1.3051>
   rotation=<rot 0 -87.6309 -73.7335>
  }]
 dyn_keyChain_b_7 [{
   shapeType=Capsule
   length=4.8707
   radius=3.4124
   shapeName=dyn_keyChain_b_7_capsule
   center=<3.6757 -1.6821 -0.0408>
   rotation=<rot 0.9055 -92.8178 -68.467>
  }]
 dyn_keyChain_f_1 [{
   shapeType=Capsule
   length=5.6301
   radius=1.5711
   shapeName=dyn_keyChain_f_1_capsule
   center=<4.7943 -0.9233 0.0846>
   rotation=<rot -90 180 -177.5214>
  }]
 dyn_keyChain_f_2 [{
   shapeType=Capsule
   length=4.8618
   radius=3.6259
   shapeName=dyn_keyChain_f_2_capsule
   center=<3.6392 -1.8126 -0.2968>
   rotation=<rot -11.8226 -90.0001 -4.8042>
  }]
 dyn_keyChain_f_3 [{
   shapeType=Capsule
   length=6.3516
   radius=3.8474
   shapeName=dyn_keyChain_f_3_capsule
   center=<6.1503 -1.85 -0.0642>
   rotation=<rot 0.0798 -87.8919 -89.0205>
  }]
 dyn_keyChain_f_4 [{
   shapeType=Capsule
   length=7.44
   radius=3.4508
   shapeName=dyn_keyChain_f_4_capsule
   center=<3.8854 -1.7459 -0.4322>
   rotation=<rot -0.0612 -89.2676 -89.8748>
  }]
 dyn_keyChain_f_5 [{
   shapeType=Capsule
   length=8.208
   radius=3.7047
   shapeName=dyn_keyChain_f_5_capsule
   center=<4.712 -2.8998 -0.431>
   rotation=<rot 0 -90.1909 -89.4177>
  }]
 floor [{shapeType=Capsule, length=0, radius=1000, shapeName=floor_capsule, center=<0 0 -1000>}]
 pelvis [{shapeType=Capsule, length=9.6411, radius=10.2033, shapeName=pelvis_capsule, center=<-4.442 -1.8572 0>, rotation=<rot 0 -90 0>}]
 spine_01 [{
   shapeType=Capsule
   length=6.2398
   radius=7.8297
   shapeName=spine_01_capsule
   center=<1.3791 1.1186 0.0229>
   rotation=<rot -2.5247 -90 1.1766>
  }]
 thigh_r [
  {
   shapeType=Capsule
   length=16.5208
   radius=7.6677
   shapeName=thigh_r_capsule
   center=<12.6623 0.5436 -0.6116>
   rotation=<rot -88.6886 3.563 -7.3013>
  }
  {
   shapeType=Capsule
   length=16.7292
   radius=7.1397
   shapeName=thigh_r_capsule2
   center=<30.6812 -0.3707 -1.1831>
   rotation=<rot -88.9063 -138.131 135.9469>
  }
 ]
}
constraints [
 {
  bone1Name=spine_01
  bone2Name=pelvis
  swing1LimitDegrees=45
  swing1Motion=Free
  swing2LimitDegrees=45
  swing2Motion=Free
  twistLimitDegrees=45
  twistMotion=Free
 }
 {
  bone1Name=thigh_r
  bone2Name=pelvis
  swing1LimitDegrees=45
  swing1Motion=Free
  swing2LimitDegrees=45
  swing2Motion=Free
  twistLimitDegrees=45
  twistMotion=Free
 }
 {
  bone1Name=base_keyChain_b_1
  bone2Name=spine_01
  swing1LimitDegrees=45
  swing1Motion=Free
  swing2LimitDegrees=45
  swing2Motion=Free
  twistLimitDegrees=45
  twistMotion=Free
 }
 {bone1Name=dyn_keyChain_b_1, bone2Name=base_keyChain_b_1, swing1LimitDegrees=10, swing2LimitDegrees=10, twistLimitDegrees=0}
 {bone1Name=dyn_keyChain_b_2, bone2Name=dyn_keyChain_b_1, swing1LimitDegrees=20, swing2LimitDegrees=10, twistLimitDegrees=0}
 {bone1Name=dyn_keyChain_b_3, bone2Name=dyn_keyChain_b_2, swing1LimitDegrees=15, swing2LimitDegrees=0, twistLimitDegrees=0}
 {bone1Name=dyn_keyChain_b_4, bone2Name=dyn_keyChain_b_3, swing1LimitDegrees=15, swing2LimitDegrees=0, twistLimitDegrees=0}
 {bone1Name=dyn_keyChain_b_5, bone2Name=dyn_keyChain_b_4, swing1LimitDegrees=15, swing2LimitDegrees=0, twistLimitDegrees=0}
 {bone1Name=dyn_keyChain_b_6, bone2Name=dyn_keyChain_b_5, swing1LimitDegrees=20, swing2LimitDegrees=10, twistLimitDegrees=0}
 {bone1Name=dyn_keyChain_b_7, bone2Name=dyn_keyChain_b_6, swing1LimitDegrees=10, swing2LimitDegrees=10, twistLimitDegrees=0}
 {
  bone1Name=base_keyChain_f_1
  bone2Name=spine_01
  swing1LimitDegrees=45
  swing1Motion=Free
  swing2LimitDegrees=45
  swing2Motion=Free
  twistLimitDegrees=45
  twistMotion=Free
 }
 {bone1Name=dyn_keyChain_f_1, bone2Name=base_keyChain_f_1, swing1LimitDegrees=10, swing2LimitDegrees=10, twistLimitDegrees=0}
 {bone1Name=dyn_keyChain_f_2, bone2Name=dyn_keyChain_f_1, swing1LimitDegrees=10, swing2LimitDegrees=10, twistLimitDegrees=0}
 {bone1Name=dyn_keyChain_f_3, bone2Name=dyn_keyChain_f_2, swing1LimitDegrees=10, swing2LimitDegrees=10, twistLimitDegrees=0}
 {bone1Name=dyn_keyChain_f_4, bone2Name=dyn_keyChain_f_3, swing1LimitDegrees=10, swing2LimitDegrees=10, twistLimitDegrees=0}
 {bone1Name=dyn_keyChain_f_5, bone2Name=dyn_keyChain_f_4, swing1LimitDegrees=10, swing2LimitDegrees=10, twistLimitDegrees=0}
 {
  bone1Name=dyn_keyChain_b_7
  bone2Name=dyn_keyChain_f_5
  swing1LimitDegrees=45
  swing1Motion=Locked
  swing2LimitDegrees=45
  swing2Motion=Locked
  twistLimitDegrees=45
  twistMotion=Locked
 }
]
