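# Post-process style AnimBP for the F_MED_PeonyBellow cosmetic body, driving secondary motion on collar, jacket, shoes and keychains.
# Copied parent pose feeds a Control Rig, then cached stages of bone fixes, AnimDynamics/RigidBody simulation, and post-sim constraints.
# Keychain simulation switches physics assets while emoting; output can blend in collar and backpack-attach offsets.
# EventGraph derives weights from movement, glider, weapon and backbling state, swapping skin-weight profiles by LOD and rigid-body availability.
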AnimBlueprint /BRCosmetics/Characters/Player/Female/Medium/Bodies/F_MED_PeonyBellow/Meshes/F_MED_PeonyBellow_AnimBP
target_skeleton=&/F_MED_PeonyBellow_Skeleton
graphs [
 AnimGraph {nodes [
   n1 = "Transform (Modify) Bone - Bone: dyn_neckCollar_l_3"(
    ComponentPose=Local To Component(LocalPose=Use cached pose 'postSimFixes'())
    Alpha=1
    Translation="0,0,-2.5"
    Rotation="0,0,0"
    Scale="0.6,0.6,0.6"
   ) {
    scaleMode=BMM_Additive
    scaleSpace=BCS_BoneSpace
    translationMode=BMM_Additive
    translationSpace=BCS_BoneSpace
    alphaBoolBlend {blendOption=Linear}
    boneToModify {boneName=dyn_neckCollar_l_3}
    scale=<0.6 0.6 0.6>
    translation {z=-2.5}
   }
   n2 = "Transform (Modify) Bone - Bone: dyn_neckCollar_c_1"(ComponentPose=@n1, Alpha=1, Translation="0,0,-2.5", Rotation="0,0,0", Scale="0.6,0.6,0.6") {
    scaleMode=BMM_Additive
    scaleSpace=BCS_BoneSpace
    translationMode=BMM_Additive
    translationSpace=BCS_BoneSpace
    alphaBoolBlend {blendOption=Linear}
    boneToModify {boneName=dyn_neckCollar_c_1}
    scale=<0.6 0.6 0.6>
    translation {z=-2.5}
   }
   n3 = "Transform (Modify) Bone - Bone: dyn_neckCollar_r_3"(ComponentPose=@n2, Alpha=1, Translation="0,0,-2.5", Rotation="0,0,0", Scale="0.6,0.6,0.6") {
    scaleMode=BMM_Additive
    scaleSpace=BCS_BoneSpace
    translationMode=BMM_Additive
    translationSpace=BCS_BoneSpace
    alphaBoolBlend {blendOption=Linear}
    boneToModify {boneName=dyn_neckCollar_r_3}
    scale=<0.6 0.6 0.6>
    translation {z=-2.5}
   }
   n4 = "Transform (Modify) Bone - Bone: attach_backpack"(ComponentPose=@n3, Alpha=1, Translation="0,-1,0", Rotation="0,0,-8", Scale="1,1,1") {
    rotationMode=BMM_Additive
    rotationSpace=BCS_BoneSpace
    scaleMode=BMM_Additive
    scaleSpace=BCS_BoneSpace
    translationMode=BMM_Additive
    translationSpace=BCS_BoneSpace
    alphaBoolBlend {blendOption=Linear}
    boneToModify {boneName=attach_backpack}
    rotation {roll=-8}
    translation {y=-1}
   }
   Output Pose(Result=Blend Poses by bool(BlendPose_0=Component To Local(ComponentPose=@n4), BlendPose_1=Use cached pose 'postSimFixes'()) {blendType=Linear}) {layerGroup=None}
   n5 = Control Rig!(
    Source=Copy Pose From Mesh() {bUseAttachedParent=true}
    Alpha
    pub isCrouching:double=Get isCrouching float() -> (isCrouching float:double)
    pub isTargeting:double=Get isTargeting float() -> (isTargeting float:double)
    pub isGunOut:double=Get isGunOut() -> (isGunOut:double)
    pub isBaseBallSliding float:double=Get isbaseBallSliding float() -> (isbaseBallSliding float:double)
    pub isKneeSliding float:double=Get iskneeSliding float() -> (iskneeSliding float:double)
   ) -> (Pose) {
    alphaBoolBlend {blendOption=Linear}
    controlRigAssetReference {blueprintRigClass=&/F_MED_PeonyBellow_CtrlRig.F_MED_PeonyBellow_CtrlRig_C}
   }
   Save cached pose 'controlRigCP'(Pose=@n5) {cacheName=controlRigCP}
   n6 = Anim Dynamics - Bone: dyn_shoe_handle_l!(ComponentPose=Local To Component(LocalPose=Use cached pose 'generalAdjuistementsCP'()), Alpha=1, GravityOverride:Vector="0,0,-980") -> (Pose) {
    angularDampingOverride=1
    angularSpringConstant=8000
    bAngularSpring=true
    bOverrideAngularDamping=true
    bUseGravityOverride=true
    gravityScale=0.35
    simulationSpace=RootRelative
    alphaBoolBlend {blendOption=Linear}
    boundBone {boneName=dyn_shoe_handle_l}
    componentAppliedLinearAccClamp=<3000 3000 3000>
    componentLinearAccScale=<1 1 1>
    componentLinearVelScale=<1 1 1>
    gravityOverride {z=-980}
    physicsBodyDefinitions [{
      boundBone {boneName=dyn_shoe_handle_l}
      boxExtents=<20 20 20>
      constraintSetup {angularLimitsMax=<0 15 40>, angularLimitsMin=<0 -15 -20>, angularTarget=<1 0 0>}
      localJointOffset=<5 0 0>
     }]
   }
   n7 = Anim Dynamics - Bone: dyn_shoe_handle_r!(ComponentPose=@n6, Alpha=1, GravityOverride:Vector="0,0,-980") -> (Pose) {
    angularDampingOverride=1
    angularSpringConstant=8000
    bAngularSpring=true
    bOverrideAngularDamping=true
    bUseGravityOverride=true
    gravityScale=0.35
    simulationSpace=RootRelative
    alphaBoolBlend {blendOption=Linear}
    boundBone {boneName=dyn_shoe_handle_r}
    componentAppliedLinearAccClamp=<3000 3000 3000>
    componentLinearAccScale=<1 1 1>
    componentLinearVelScale=<1 1 1>
    gravityOverride {z=-980}
    physicsBodyDefinitions [{
      boundBone {boneName=dyn_shoe_handle_r}
      boxExtents=<20 20 20>
      constraintSetup {angularLimitsMax=<0 15 40>, angularLimitsMin=<0 -15 -20>, angularTarget=<1 0 0>}
      localJointOffset=<5 0 0>
     }]
   }
   n8 = Anim Dynamics - Bone: base_jacket_corrective_l!(ComponentPose=@n7, Alpha=1, GravityOverride:Vector="0,0,-980") -> (Pose) {
    angularDampingOverride=1
    angularSpringConstant=8000
    bAngularSpring=true
    bOverrideAngularDamping=true
    gravityScale=0.35
    simulationSpace=BoneRelative
    alphaBoolBlend {blendOption=Linear}
    boundBone {boneName=base_jacket_corrective_l}
    componentAppliedLinearAccClamp=<3000 3000 3000>
    componentLinearAccScale=<1 1 1>
    componentLinearVelScale=<1 1 1>
    gravityOverride {z=-980}
    physicsBodyDefinitions [{
      boundBone {boneName=base_jacket_corrective_l}
      boxExtents=<22 2 22>
      constraintSetup {angularLimitsMax=<0 5 0>, angularLimitsMin=<0 -5 0>, angularTarget=<1 0 0>}
      localJointOffset=<4 4 0>
     }]
    relativeSpaceBone {boneName=clavicle_l}
   }
   n9 = Anim Dynamics - Bone: base_jacket_corrective_r!(ComponentPose=@n8, Alpha=1, GravityOverride:Vector="0,0,-980") -> (Pose) {
    angularDampingOverride=1
    angularSpringConstant=8000
    bAngularSpring=true
    bOverrideAngularDamping=true
    gravityScale=0.35
    simulationSpace=BoneRelative
    alphaBoolBlend {blendOption=Linear}
    boundBone {boneName=base_jacket_corrective_r}
    componentAppliedLinearAccClamp=<3000 3000 3000>
    componentLinearAccScale=<1 1 1>
    componentLinearVelScale=<1 1 1>
    gravityOverride {z=-980}
    physicsBodyDefinitions [{
      boundBone {boneName=base_jacket_corrective_r}
      boxExtents=<25 25 25>
      constraintSetup {angularLimitsMax=<0 5 0>, angularLimitsMin=<0 -5 0>, angularTarget=<1 0 0>}
      localJointOffset=<4 4 0>
     }]
    relativeSpaceBone {boneName=clavicle_r}
   }
   n10 = Anim Dynamics - Bone: dyn_shoeLace_l_A!(ComponentPose=@n9, Alpha=1, GravityOverride:Vector="0,0,-980") -> (Pose) {
    angularDampingOverride=1
    angularSpringConstant=12000
    bAngularSpring=true
    bOverrideAngularDamping=true
    bUseGravityOverride=true
    gravityScale=0.35
    simulationSpace=RootRelative
    alphaBoolBlend {blendOption=Linear}
    boundBone {boneName=dyn_shoeLace_l_A}
    componentAppliedLinearAccClamp=<4000 4000 4000>
    componentLinearAccScale=<1 1 1>
    componentLinearVelScale=<1 1 1>
    gravityOverride {z=-980}
    physicsBodyDefinitions [{
      boundBone {boneName=dyn_shoeLace_l_A}
      boxExtents=<20 20 20>
      constraintSetup {angularLimitsMax=<0 0 7>, angularLimitsMin=<0 -15 -7>, angularTarget=<1 0 0>}
      localJointOffset=<5 0 0>
     }]
   }
   n11 = Anim Dynamics - Bone: dyn_shoeLace_l_B!(ComponentPose=@n10, Alpha=1, GravityOverride:Vector="0,0,-980") -> (Pose) {
    angularDampingOverride=1
    angularSpringConstant=12000
    bAngularSpring=true
    bOverrideAngularDamping=true
    bUseGravityOverride=true
    gravityScale=0.35
    simulationSpace=RootRelative
    alphaBoolBlend {blendOption=Linear}
    boundBone {boneName=dyn_shoeLace_l_B}
    componentAppliedLinearAccClamp=<4000 4000 4000>
    componentLinearAccScale=<1 1 1>
    componentLinearVelScale=<1 1 1>
    gravityOverride {z=-980}
    physicsBodyDefinitions [{
      boundBone {boneName=dyn_shoeLace_l_B}
      boxExtents=<20 20 20>
      constraintSetup {angularLimitsMax=<0 0 7>, angularLimitsMin=<0 -15 -7>, angularTarget=<1 0 0>}
      localJointOffset=<5 0 0>
     }]
   }
   n12 = Anim Dynamics - Bone: dyn_shoeLace_r_B!(ComponentPose=@n11, Alpha=1, GravityOverride:Vector="0,0,-980") -> (Pose) {
    angularDampingOverride=1
    angularSpringConstant=12000
    bAngularSpring=true
    bOverrideAngularDamping=true
    bUseGravityOverride=true
    gravityScale=0.35
    simulationSpace=RootRelative
    alphaBoolBlend {blendOption=Linear}
    boundBone {boneName=dyn_shoeLace_r_B}
    componentAppliedLinearAccClamp=<4000 4000 4000>
    componentLinearAccScale=<1 1 1>
    componentLinearVelScale=<1 1 1>
    gravityOverride {z=-980}
    physicsBodyDefinitions [{
      boundBone {boneName=dyn_shoeLace_r_B}
      boxExtents=<20 20 20>
      constraintSetup {angularLimitsMax=<0 0 7>, angularLimitsMin=<0 -15 -7>, angularTarget=<1 0 0>}
      localJointOffset=<5 0 0>
     }]
   }
   n13 = Anim Dynamics - Bone: dyn_shoeLace_r_A!(ComponentPose=@n12, Alpha=1, GravityOverride:Vector="0,0,-980") -> (Pose) {
    angularDampingOverride=1
    angularSpringConstant=12000
    bAngularSpring=true
    bOverrideAngularDamping=true
    bUseGravityOverride=true
    gravityScale=0.35
    simulationSpace=RootRelative
    alphaBoolBlend {blendOption=Linear}
    boundBone {boneName=dyn_shoeLace_r_A}
    componentAppliedLinearAccClamp=<4000 4000 4000>
    componentLinearAccScale=<1 1 1>
    componentLinearVelScale=<1 1 1>
    gravityOverride {z=-980}
    physicsBodyDefinitions [{
      boundBone {boneName=dyn_shoeLace_r_A}
      boxExtents=<20 20 20>
      constraintSetup {angularLimitsMax=<0 0 7>, angularLimitsMin=<0 -15 -7>, angularTarget=<1 0 0>}
      localJointOffset=<5 0 0>
     }]
   }
   n14 = RigidBody!(ComponentPose=@n13, Alpha=0.5, OverrideWorldGravity:Vector="0,0,980") -> (Pose) {
    bOverrideWorldGravity=true
    overridePhysicsAsset=&/F_MED_PeonyBellow_collar_Physics
    simulationSpace=BaseBoneSpace
    alphaBoolBlend {blendOption=Linear}
    alphaScaleBiasClamp {bInterpResult=true, interpSpeedDecreasing=0.5, interpSpeedIncreasing=0.5}
    baseBoneRef {boneName=spine_05}
    componentAppliedLinearAccClamp=<3000 3000 3000>
    overrideWorldGravity {z=980}
    simSpaceSettings {maxAngularAcceleration=1, maxAngularVelocity=3, maxLinearAcceleration=500, maxLinearVelocity=500, worldAlpha=1}
   }
   Save cached pose 'dynamicsCP'(Pose=Component To Local(ComponentPose=@n14)) {cacheName=dynamicsCP}
   n15 = RigidBody(ComponentPose=Local To Component(LocalPose=Use cached pose 'dynamicsCP'()), OverrideWorldGravity="0,0,-980") {
    bClampLinearTranslationLimitToRefPose=true
    bOverrideWorldGravity=true
    overridePhysicsAsset=&/F_MED_PeonyBellow_keyChain_Physics
    simulationSpace=BaseBoneSpace
    alphaBoolBlend {blendOption=Linear}
    baseBoneRef {boneName=base_keyChain_SimSpace}
    componentAppliedLinearAccClamp=<3000 3000 3000>
    simSpaceSettings {
     maxAngularAcceleration=1
     maxAngularVelocity=3
     maxLinearAcceleration=1000
     maxLinearVelocity=500
     worldAlpha=1
     externalLinearDragV=<0.5 0.5 0.5>
    }
   }
   n16 = Component To Local(ComponentPose=@n15)
   n17 = RigidBody(ComponentPose=Local To Component(LocalPose=Use cached pose 'dynamicsCP'()), OverrideWorldGravity="0,0,-980") {
    bClampLinearTranslationLimitToRefPose=true
    bOverrideWorldGravity=true
    overridePhysicsAsset=&/F_MED_PeonyBellow_keyChain_Emoting_Physics
    simulationSpace=BaseBoneSpace
    alphaBoolBlend {blendOption=Linear}
    baseBoneRef {boneName=thigh_twist_01_r}
    componentAppliedLinearAccClamp=<3000 3000 3000>
    simSpaceSettings {
     maxAngularAcceleration=1
     maxAngularVelocity=3
     maxLinearAcceleration=1000
     maxLinearVelocity=500
     worldAlpha=1
     externalLinearDragV=<0.5 0.5 0.5>
    }
   }
   n18 = Component To Local(ComponentPose=@n17)
   n19 = Get bIsPlayingEmote()
   n20 = RigidBody(
    ComponentPose=Local To Component(LocalPose=Blend Poses by bool(BlendPose_0=@n18, BlendPose_1=@n16, bActiveValue=@n19) {blendType=Linear})
    OverrideWorldGravity="0,0,-980"
   ) {
    bClampLinearTranslationLimitToRefPose=true
    bOverrideWorldGravity=true
    overridePhysicsAsset=&/F_MED_PeonyBellow_Physics
    simulationSpace=BaseBoneSpace
    alphaBoolBlend {blendOption=Linear}
    baseBoneRef {boneName=spine_04}
    componentAppliedLinearAccClamp=<3000 3000 3000>
    simSpaceSettings {
     maxAngularAcceleration=1
     maxAngularVelocity=3
     maxLinearAcceleration=1500
     maxLinearVelocity=500
     worldAlpha=1
     externalLinearDragV=<0.2 0.2 0.2>
    }
   }
   Save cached pose 'dynamicsKeyChainsCP'(Pose=Component To Local(ComponentPose=@n20)) {cacheName=dynamicsKeyChainsCP}
   n21 = "Transform (Modify) Bone - Bone: attach_cape"!(
    ComponentPose=Local To Component(LocalPose=Use cached pose 'presimFixesCP'())
    bAlphaBoolEnabled:bool=True
    Translation="0,-1,0"
    Rotation="0,0,16"
   ) -> (Pose) {
    alphaInputType=Bool
    rotationMode=BMM_Additive
    rotationSpace=BCS_BoneSpace
    translationMode=BMM_Additive
    translationSpace=BCS_BoneSpace
    alphaBoolBlend {blendInTime=0.5, blendOption=CircularOut, blendOutTime=0.5}
    boneToModify {boneName=attach_cape}
    rotation {roll=16}
    translation {y=-1}
   }
   n22 = "Transform (Modify) Bone - Bone: attach_cape"(ComponentPose=@n21, bAlphaBoolEnabled=True, Translation="0,-2,0", Rotation="0,0,20") {
    alphaInputType=Bool
    rotationMode=BMM_Additive
    rotationSpace=BCS_BoneSpace
    translationMode=BMM_Additive
    translationSpace=BCS_BoneSpace
    alphaBoolBlend {blendInTime=0.5, blendOption=CircularOut, blendOutTime=0.5}
    boneToModify {boneName=attach_cape}
    rotation {roll=20}
    translation {y=-2}
   }
   n23 = "Transform (Modify) Bone - Bone: attach_cape"(ComponentPose=@n22, bAlphaBoolEnabled=True, Translation="0,0,0", Rotation="0,0,10") {
    alphaInputType=Bool
    rotationMode=BMM_Additive
    rotationSpace=BCS_BoneSpace
    translationMode=BMM_Additive
    translationSpace=BCS_BoneSpace
    alphaBoolBlend {blendInTime=0.5, blendOption=Cubic, blendOutTime=0.5}
    boneToModify {boneName=attach_cape}
    rotation {roll=10}
   }
   n24 = "Constraint - base_keyChain_SimSpace (R)"(ComponentPose=@n23, Alpha=0.5, ConstraintWeights_0=1) {
    alphaBoolBlend {blendOption=Linear}
    boneToModify {boneName=base_keyChain_SimSpace}
    constraintSetup [{transformType=Rotation, targetBone {boneName=thigh_r}}]
    constraintWeights [1]
   }
   n25 = "Transform (Modify) Bone - Bone: base_neckCollar_r_1"!(ComponentPose=@n24, bAlphaBoolEnabled:bool=True, Translation="0,0,-3", Rotation="0,0,0") -> (Pose) {
    alphaInputType=Bool
    rotationMode=BMM_Additive
    rotationSpace=BCS_BoneSpace
    translationMode=BMM_Additive
    translationSpace=BCS_BoneSpace
    alphaBoolBlend {blendInTime=0.5, blendOption=CircularOut, blendOutTime=0.5}
    boneToModify {boneName=base_neckCollar_r_1}
    translation {z=-3}
   }
   n26 = "Transform (Modify) Bone - Bone: base_neckCollar_r_2"!(ComponentPose=@n25, bAlphaBoolEnabled:bool=True, Translation="0,0,-5", Rotation="0,0,0") -> (Pose) {
    alphaInputType=Bool
    rotationMode=BMM_Additive
    rotationSpace=BCS_BoneSpace
    translationMode=BMM_Additive
    translationSpace=BCS_BoneSpace
    alphaBoolBlend {blendInTime=0.5, blendOption=CircularOut, blendOutTime=0.5}
    boneToModify {boneName=base_neckCollar_r_2}
    translation {z=-5}
   }
   n27 = "Transform (Modify) Bone - Bone: upperarm_twist_01_r"!(ComponentPose=@n26, bAlphaBoolEnabled:bool=True, Translation="0,0,0", Rotation="0,0,10", Scale="1,1.3,1.3") -> (Pose) {
    alphaInputType=Bool
    rotationMode=BMM_Additive
    rotationSpace=BCS_BoneSpace
    scaleMode=BMM_Additive
    scaleSpace=BCS_BoneSpace
    translationMode=BMM_Additive
    translationSpace=BCS_BoneSpace
    alphaBoolBlend {blendInTime=0.5, blendOption=Cubic, blendOutTime=0.5}
    boneToModify {boneName=upperarm_twist_01_r}
    rotation {roll=10}
    scale {y=1.3, z=1.3}
   }
   Save cached pose 'generalAdjuistementsCP'(Pose=Component To Local(ComponentPose=@n27)) {cacheName=generalAdjuistementsCP}
   n28 = "Transform (Modify) Bone - Bone: kin_jacket_all"!(
    ComponentPose=Local To Component(LocalPose=Use cached pose 'controlRigCP'())
    bAlphaBoolEnabled:bool=True
    Translation="0,0,-1"
    Rotation="10,0,10"
   ) -> (Pose) {
    alphaInputType=Bool
    rotationMode=BMM_Additive
    rotationSpace=BCS_BoneSpace
    translationMode=BMM_Additive
    translationSpace=BCS_BoneSpace
    alphaBoolBlend {blendInTime=0.2, blendOption=Cubic, blendOutTime=0.2}
    boneToModify {boneName=kin_jacket_all}
    rotation {pitch=10, roll=10}
    translation {z=-1}
   }
   n29 = "Transform (Modify) Bone - Bone: base_jacket_fr_r"!(ComponentPose=@n28, bAlphaBoolEnabled:bool=True, Translation="-5,-2,0", Rotation="0,-30,0") -> (Pose) {
    alphaInputType=Bool
    rotationMode=BMM_Additive
    rotationSpace=BCS_BoneSpace
    translationMode=BMM_Additive
    translationSpace=BCS_BoneSpace
    alphaBoolBlend {blendInTime=0.2, blendOption=Cubic, blendOutTime=0.2}
    boneToModify {boneName=base_jacket_fr_r}
    rotation {yaw=-30}
    translation=<-5 -2>
   }
   n30 = "Transform (Modify) Bone - Bone: base_neckCollar_l_1"(ComponentPose=@n29, Alpha=1, Translation="0,0,2.8", Rotation="40,0,0", Scale="1,1,0.8") {
    rotationMode=BMM_Additive
    rotationSpace=BCS_BoneSpace
    scaleMode=BMM_Additive
    scaleSpace=BCS_BoneSpace
    translationMode=BMM_Additive
    translationSpace=BCS_BoneSpace
    alphaBoolBlend {blendOption=Linear}
    alphaScaleBiasClamp {bInterpResult=true, interpSpeedDecreasing=2, interpSpeedIncreasing=2}
    boneToModify {boneName=base_neckCollar_l_1}
    rotation {pitch=40}
    scale {z=0.8}
    translation {z=2.8}
   }
   n31 = "Transform (Modify) Bone - Bone: base_neckCollar_l_2"(ComponentPose=@n30, Alpha=1, Translation="0,0,0", Rotation="0,0,0", Scale="1,1,0.5") {
    rotationMode=BMM_Additive
    rotationSpace=BCS_BoneSpace
    scaleMode=BMM_Additive
    scaleSpace=BCS_BoneSpace
    translationMode=BMM_Additive
    translationSpace=BCS_BoneSpace
    alphaBoolBlend {blendOption=Linear}
    alphaScaleBiasClamp {bInterpResult=true, interpSpeedDecreasing=2, interpSpeedIncreasing=2}
    boneToModify {boneName=base_neckCollar_l_2}
    scale {z=0.5}
   }
   n32 = "Transform (Modify) Bone - Bone: base_neckCollar_l_3"(ComponentPose=@n31, Alpha=1, Translation="-3.1,0,0", Rotation="0,0,0", Scale="1,1,1") {
    rotationMode=BMM_Additive
    rotationSpace=BCS_BoneSpace
    scaleMode=BMM_Additive
    scaleSpace=BCS_BoneSpace
    translationMode=BMM_Additive
    translationSpace=BCS_BoneSpace
    alphaBoolBlend {blendOption=Linear}
    alphaScaleBiasClamp {bInterpResult=true, interpSpeedDecreasing=2, interpSpeedIncreasing=2}
    boneToModify {boneName=base_neckCollar_l_3}
    translation {x=-3.1}
   }
   n33 = "Transform (Modify) Bone - Bone: base_neckCollar_c_1"(ComponentPose=@n32, Alpha=1, Translation="-0.7,0,0", Rotation="0,0,0", Scale="1,1,1") {
    rotationMode=BMM_Additive
    rotationSpace=BCS_BoneSpace
    scaleMode=BMM_Additive
    scaleSpace=BCS_BoneSpace
    translationMode=BMM_Additive
    translationSpace=BCS_BoneSpace
    alphaBoolBlend {blendOption=Linear}
    alphaScaleBiasClamp {bInterpResult=true, interpSpeedDecreasing=2, interpSpeedIncreasing=2}
    boneToModify {boneName=base_neckCollar_c_1}
    translation {x=-0.7}
   }
   n34 = "Transform (Modify) Bone - Bone: base_neckCollar_r_1"!(ComponentPose=@n33, Alpha:float=1, Translation="0,0,2.8", Rotation="0,0,0", Scale:Vector="1,1,1") -> (Pose) {
    rotationMode=BMM_Additive
    rotationSpace=BCS_BoneSpace
    scaleMode=BMM_Additive
    scaleSpace=BCS_BoneSpace
    translationMode=BMM_Additive
    translationSpace=BCS_BoneSpace
    alphaBoolBlend {blendOption=Linear}
    alphaScaleBiasClamp {bInterpResult=true, interpSpeedDecreasing=2, interpSpeedIncreasing=2}
    boneToModify {boneName=base_neckCollar_r_1}
    translation {z=2.8}
   }
   n35 = "Transform (Modify) Bone - Bone: base_neckCollar_r_2"!(ComponentPose=@n34, Alpha:float=1, Translation="0,0,0", Rotation="0,0,0", Scale:Vector="1,1,0.5") -> (Pose) {
    rotationMode=BMM_Additive
    rotationSpace=BCS_BoneSpace
    scaleMode=BMM_Additive
    scaleSpace=BCS_BoneSpace
    translationMode=BMM_Additive
    translationSpace=BCS_BoneSpace
    alphaBoolBlend {blendOption=Linear}
    alphaScaleBiasClamp {bInterpResult=true, interpSpeedDecreasing=2, interpSpeedIncreasing=2}
    boneToModify {boneName=base_neckCollar_r_2}
    scale {z=0.5}
   }
   n36 = "Transform (Modify) Bone - Bone: base_neckCollar_r_3"(ComponentPose=@n35, Alpha=1, Translation="-3.1,0,0", Rotation="0,0,0", Scale="1,1,1") {
    rotationMode=BMM_Additive
    rotationSpace=BCS_BoneSpace
    scaleMode=BMM_Additive
    scaleSpace=BCS_BoneSpace
    translationMode=BMM_Additive
    translationSpace=BCS_BoneSpace
    alphaBoolBlend {blendOption=Linear}
    alphaScaleBiasClamp {bInterpResult=true, interpSpeedDecreasing=2, interpSpeedIncreasing=2}
    boneToModify {boneName=base_neckCollar_r_3}
    translation {x=-3.1}
   }
   Save cached pose 'presimFixesCP'(Pose=Component To Local(ComponentPose=@n36)) {cacheName=presimFixesCP}
   n37 = Look At - Bone: aim_keyChain_b_3_A(ComponentPose=Local To Component(LocalPose=Use cached pose 'dynamicsKeyChainsCP'()), Alpha=1) {
    alphaBoolBlend {blendOption=Linear}
    boneToModify {boneName=aim_keyChain_b_3_A}
    lookAtLocation {x=0}
    lookAtTarget {boneReference {boneName=target_keyChain_b_3_A}}
    lookAt_Axis {axis=<1 0>}
    lookUp_Axis {axis {x=0, z=1}}
   }
   n38 = "Constraint - attach_cape (PPPPPPPP)"!(
    ComponentPose=@n37
    Alpha=1
    ConstraintWeights_0=0.5035
    ConstraintWeights_1:float=0.2072
    ConstraintWeights_2:float=0.2072
    ConstraintWeights_3:float=0.0423
    ConstraintWeights_4:float=0.0143
    ConstraintWeights_5:float=0.0143
    ConstraintWeights_6:float=0.0056
    ConstraintWeights_7:float=0.0056
   ) -> (Pose) {
    alphaBoolBlend {blendOption=Linear}
    boneToModify {boneName=attach_cape}
    constraintSetup [
     {targetBone {boneName=dyn_neckCollar_c_1}}
     {targetBone {boneName=dyn_neckCollar_l_3}}
     {targetBone {boneName=dyn_neckCollar_r_3}}
     {targetBone {boneName=spine_05}}
     {targetBone {boneName=dyn_neckCollar_l_2}}
     {targetBone {boneName=dyn_neckCollar_r_2}}
     {targetBone {boneName=clavicle_r}}
     {targetBone {boneName=clavicle_l}}
    ]
    constraintWeights [1,1,1,1,1,1,1,1]
   }
   n39 = Look At - Bone: attach_cape(ComponentPose=@n38, Alpha=1) {
    alphaBoolBlend {blendOption=Linear}
    boneToModify {boneName=attach_cape}
    lookAtLocation=<-3.8529 -17.9995 0>
    lookAtTarget {boneReference {boneName=spine_03}}
    lookAt_Axis {axis {y=0, z=-1}}
    lookUp_Axis {axis {x=0, z=1}}
   }
   n40 = "Transform (Modify) Bone - Bone: dyn_keyChain_b_1"!(ComponentPose=@n39, Alpha=1, Translation="0,0,0", Rotation="0,-10,0") -> (Pose) {
    rotationMode=BMM_Additive
    rotationSpace=BCS_BoneSpace
    translationMode=BMM_Additive
    translationSpace=BCS_BoneSpace
    alphaBoolBlend {blendOption=Linear}
    alphaScaleBiasClamp {bInterpResult=true, interpSpeedDecreasing=5, interpSpeedIncreasing=5}
    boneToModify {boneName=dyn_keyChain_b_1}
    rotation {yaw=-10}
   }
   n41 = "Transform (Modify) Bone - Bone: dyn_keyChain_f_1"!(ComponentPose=@n40, Alpha=1, Translation="0,0,0", Rotation="7.050,-4.823,-2.388") -> (Pose) {
    rotationMode=BMM_Additive
    rotationSpace=BCS_BoneSpace
    translationMode=BMM_Additive
    translationSpace=BCS_BoneSpace
    alphaBoolBlend {blendOption=Linear}
    alphaScaleBiasClamp {bInterpResult=true, interpSpeedDecreasing=5, interpSpeedIncreasing=5}
    boneToModify {boneName=dyn_keyChain_f_1}
    rotation=<rot 7.05 -4.8231 -2.388>
   }
   Save cached pose 'postSimFixes'(Pose=Component To Local(ComponentPose=@n41)) {cacheName=postSimFixes}
  ]}
 EventGraph {nodes [
   Event BlueprintInitializeAnimation()
   > Set SkelMeshOwn(SkelMeshOwn:Skeletal Mesh Component Object Reference=GetPartSkeletalMeshComponent()) -> (Output_Get:Skeletal Mesh Component Object Reference)
   Event OnParachuteOpened()
   > n1 = Switch on EGliderType(Selection=Get PlayerGliderType() -> (PlayerGliderType=HangGlider))
   > Set isUsingHangGlidingInParachute(isUsingHangGlidingInParachute:bool=true) -> (Output_Get:bool)
   Set isUsingUmbrellaParachute(execute=@n1.Umbrella, isUsingUmbrellaParachute:bool=true) -> (Output_Get:bool)
   Event OnParachuteClosed()
   > Set isUsingHangGlidingInParachute(isUsingHangGlidingInParachute=false)
   > Set isUsingUmbrellaParachute(isUsingUmbrellaParachute=false)
   > ResetDynamics(InTeleportType=TeleportPhysics)
   > ResetDynamics(InTeleportType=ResetPhysics)
   n2 = Event OnCharacterCustomizationLoaded()
   > Set SkelMeshBackpack(SkelMeshBackpack:Skeletal Mesh Component Budgeted Object Reference=Get Array Item with EFort Custom Part Type(Array=@n2.CustomMeshParts, PartType=Backpack)) -> (Output_Get:Skeletal Mesh Component Budgeted Object Reference)
   n3 = MatchesWildcard(
    SourceString=GetDisplayName(Object=Get SkelMeshBackpack() -> (SkelMeshBackpack:Skeletal Mesh Component Budgeted Object Reference))
    Wildcard=*F_MED_PeonyBellow_Pack
    SearchCase=CaseSensitive
   )
   n4 = CheckForBackbling()
   > Set isWearingValidBP(isWearingValidBP:bool=AND Boolean(A=OR Boolean!(A=@n4, B=@n4.bIsBackblingACape, C:bool=@n4.bIsBackblingLong) -> (ReturnValue), B=NOT Boolean(A=@n3))) -> (Output_Get:bool)
   Event OnWeaponChanged()
   > Set isGunOut(isGunOut:double=1) -> (Output_Get:double)
   > n5 = Switch on EFortWeaponCoreAnimation!(execute, Selection:EFortWeaponCoreAnimation Enum=Get WeaponCoreAnim!() -> (WeaponCoreAnim)) -> (
    Melee:Exec
    Pistol:Exec
    Shotgun:Exec
    PaperBlueprint:Exec
    Rifle:Exec
    MeleeOneHand:Exec
    MachinePistol:Exec
    RocketLauncher:Exec
    GrenadeLauncher:Exec
    GoingCommando:Exec
    AssaultRifle:Exec
    TacticalShotgun:Exec
    SniperRifle:Exec
    TrapPlacement:Exec
    ShoulderLauncher:Exec
    AbilityDecoTool:Exec
    Crossbow:Exec
    C4:Exec
    RemoteControl:Exec
    DualWield:Exec
    AR_BullPup:Exec
    AR_ForwardGrip:Exec
    MedPackPaddles:Exec
    SMG_P90:Exec
    AR_DrumGun:Exec
    Consumable_Small:Exec
    Consumable_Large:Exec
    Balloon:Exec
    MountedTurret:Exec
    CreativeTool:Exec
    ExplosiveBow:Exec
    AshtonIndigo:Exec
    AshtonChicago:Exec
    MeleeDualWield:Exec
    Unarmed:Exec
   )
   Set isGunOut(execute=[@n5.Melee, @n5.MeleeDualWield, @n5.MeleeOneHand, @n5.Unarmed], isGunOut=0)
   n6 = Event BlueprintUpdateAnimation()
   > Set remap 60to30fps(remap 60to30fps:double=Remap Float(Input=@n6.DeltaTimeX, OldRange=(X=0.0167,Y=0.0333), NewRange=(X=0,Y=1), Clamp=true) -> (Output=~)) -> (Output_Get:double)
   > n7 = Sequence!(execute) -> (then_0, then_1, then_2:Exec, then_3:Exec, then_4:Exec)
   > n8 = Gravity Override State Function DEPRECATED(CharacterPartAnimInstance=Self-Reference(), GameTimeInSeconds=GetGameTimeInSeconds(), GravityOverrideParams=Get Gravity Override Params())
   > Set GravityOverride(GravityOverride=@n8)
   > n9 = Gravity Override State Function DEPRECATED(
    CharacterPartAnimInstance=Self-Reference()
    GameTimeInSeconds=GetGameTimeInSeconds()
    GravityOverrideParams=Get Gravity Override Params Inverse() -> (Gravity Override Params Inverse:Gravity Override Params Struct Structure)
   )
   > "Set GravityOverride  Inverse"("GravityOverride  Inverse":Vector=@n9) -> (Output_Get:Vector)
   > Set headRotationBack(headRotationBack:double=GetCurveValue(CurveName=head_backRot)) -> (Output_Get:double)
   n10 = Branch(
    execute=@n7.then_1
    Condition=OR Boolean!(A=Get bIsSkydiveFloating(), B=Get isUsingHangGlidingInParachute() -> (isUsingHangGlidingInParachute:bool)) -> (ReturnValue)
   )
   > Set weightAdjustCoatCollar(weightAdjustCoatCollar:double=1) -> (Output_Get:double)
   Set weightAdjustCoatCollar(execute=@n10.else, weightAdjustCoatCollar=0)
   n11 = Get bIsSprinting()
   n12 = Get bIsSkydiveDiveMode()
   Set moveChainPostDyn(
    execute=@n7.then_3
    moveChainPostDyn:double=float * float(A:bool=OR Boolean(A=@n11, B=@n12), B=Get remap 60to30fps() -> (remap 60to30fps:double)) {is=title K2Node_PromotableOperator}
   ) -> (Output_Get:double)
   n13 = Get bRigidBodyEnabled()
   n14 = Get CurrentLOD()
   > n15 = Branch(Condition=AND Boolean(A=NOT Boolean(A=@n13), B="integer >= integer"(A=@n14, B=1)))
   > n16 = Sequence!(execute) -> (then_0, then_1)
   > Do Once(Reset=@n18.then_0)
   n17 = Get SkelMeshOwn() -> (SkelMeshOwn:Skeletal Mesh Component Object Reference)
   > SetSkinWeightProfile!(execute, self:Skinned Mesh Component Object Reference=@n17, InProfileName:Name=Alternative, InLayer:ESkinWeightProfileLayer Enum=Secondary) -> (then, ReturnValue:bool)
   Do Once(execute=@n18.then_1, Reset=@n16.then_1)
   > ClearSkinWeightProfile!(execute, self:Skinned Mesh Component Object Reference=@n17, InLayer:ESkinWeightProfileLayer Enum=Secondary) -> (then)
   n18 = Sequence(execute=@n15.else)
   n19 = OR Boolean(A=Get bIsSkydiveFloating(), B=Get bIsSkydiveDiveMode())
   Set isSkydiveFloating float(
    execute=@n7.then_4
    isSkydiveFloating float:double=FInterpTo(
     Current=Get isSkydiveFloating float() -> (isSkydiveFloating float:double)
     Target="To Float (Boolean)"(InBool=@n19)
     DeltaTime=@n6.DeltaTimeX
     InterpSpeed=8
    )
   ) -> (Output_Get:double)
   n20 = Get bIsCrouching()
   > Set isCrouching float(isCrouching float:double=FInterpTo(Current=Get isCrouching float(), Target="To Float (Boolean)"(InBool=@n20), DeltaTime=@n6.DeltaTimeX, InterpSpeed=15)) -> (Output_Get:double)
   n21 = Get bIsTargeting()
   > Set isTargeting float(isTargeting float:double=FInterpTo(Current=Get isTargeting float(), Target="To Float (Boolean)"(InBool=@n21), DeltaTime=@n6.DeltaTimeX, InterpSpeed=15)) -> (Output_Get:double)
   n22 = Get bIsKneeSliding()
   > Set iskneeSliding float(iskneeSliding float:double=FInterpTo(Current=Get iskneeSliding float(), Target="To Float (Boolean)"(InBool=@n22), DeltaTime=@n6.DeltaTimeX, InterpSpeed=15)) -> (Output_Get:double)
   n23 = Get bIsBaseballSliding()
   > Set isbaseBallSliding float(isbaseBallSliding float:double=FInterpTo(Current=Get isbaseBallSliding float(), Target="To Float (Boolean)"(InBool=@n23), DeltaTime=@n6.DeltaTimeX, InterpSpeed=15)) -> (Output_Get:double)
   n24 = Branch(
    execute=@n7.then_2
    Condition=OR Boolean!(A=Get isUsingHangGlidingInParachute(), B=Get isUsingUmbrellaParachute() -> (isUsingUmbrellaParachute:bool), C:bool=Get bIsSkydiveFloating()) -> (ReturnValue)
   )
   > Set coatCollarRB alpha(coatCollarRB alpha:double=0.5) -> (Output_Get:double)
   Set coatCollarRB alpha(execute=@n24.else, coatCollarRB alpha=1)
  ]}
]
variables [
 coatCollarRB alpha {type=number}
 Gravity Override Params {type=GravityOverrideParamsStruct}
 Gravity Override Params Inverse {
  type=GravityOverrideParamsStruct
  default {
   dBNO {gravityOverride=<980 400 0>}
   emoteOrMelee {gravityOverride=<0 0 980>}
   front_End {gravityOverride=<0 0 980>}
   onGround_Falling {gravityOverride=<0 0 980>}
   onGround_Moving {gravityOverride=<0 0 980>}
   onGround_Standing {gravityOverride=<0 0 980>}
   rail_Grinding {gravityOverride=<0 0 980>}
   riding_Moving {gravityOverride=<0 0 980>}
   riding_Moving_Targeting {gravityOverride=<0 0 980>}
   riding_Standing {gravityOverride=<0 0 980>}
   skydive_DivingDown {jointName=root, gravityOverride=<0 0 10>, windAmplitude=<100 0 500>, windFrequency=<20 0 20>}
   skydive_FloatingDown {gravityOverride=<0 0 980>, windAmplitude=<100 0 500>, windFrequency=<20 0 20>, windOffset=<0 0 0>}
   skydive_GoingUp {gravityOverride=<0 0 980>, windAmplitude=<0 0 0>, windFrequency=<0 0 0>}
   skydive_Parachuting {gravityOverride=<0 -800 980>, windAmplitude=<10 -250 200>, windFrequency=<1.5 3 8>}
   skydive_Surfing {gravityOverride=<0 0 980>, windAmplitude=<0 0 0>, windFrequency=<0 0 0>}
   swimming {jointName=root, gravityOverride=<0 0 980>}
   tactical_Sprint {gravityOverride=<0 -250 600>}
   ziplining {gravityOverride=<0 0 980>}
  }
 }
 GravityOverride {type=Vector}
 "GravityOverride  Inverse" {type=Vector}
 headRotationBack {type=number}
 isbaseBallSliding float {type=number}
 isCrouching float {type=number}
 isGunOut {type=number}
 iskneeSliding float {type=number}
 isSkydiveFloating float {type=number}
 isTargeting float {type=number}
 isUsingHangGlidingInParachute {type=boolean}
 isUsingUmbrellaParachute {type=boolean}
 isWearingValidBP {type=boolean}
 moveChainPostDyn {type=number}
 remap 60to30fps {type=number}
 SkelMeshBackpack {type=SkeletalMeshComponentBudgeted}
 SkelMeshOwn {type=SkeletalMeshComponent}
 weightAdjustCoatCollar {type=number}
]
