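# Control Rig for PeonyBellow's face accessory skeleton that corrects hair and ponytail bones as the body and head move.
# A parent constraint blends the ponytail base mostly to spine_05, partly to the left clavicle.
# DeltaRotation converts head swing relative to head_null into a clamped weight, driving curve headRz_up_out and corrective offsets on hair strand bones.
# A debug switch, off here, copies test controls onto head, neck and clavicle bones.
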ControlRigBlueprint /BRCosmetics/Characters/Player/Female/Medium/Bodies/F_MED_PeonyBellow/Meshes/Parts/F_MED_PeonyBellow_FaceAcc_CtrlRig
graphs [
 New Function_ContainedGraph {nodes [
   n1 = RigVMFunctionEntryNode()
   RigVMFunction_ControlFlowBranch = &RigVMFunction_ControlFlowBranch(ExecuteContext=@n1.ExecuteContext, Condition=@n1.debug)
   RigVMFunctionReturnNode(ExecuteContext=@RigVMFunction_ControlFlowBranch.Completed)
   n2 = &RigUnit_ItemArray(Items=((Type=Bone,Name="head"),(Type=Bone,Name="neck_02"),(Type=Bone,Name="neck_01"),(Type=Bone,Name="clavicle_l"),(Type=Bone,Name="clavicle_r")))
   DISPATCH_RigVMDispatch_ArrayIterator = RigVMDispatchNode(Array=@n2.Items, ExecuteContext=@RigVMFunction_ControlFlowBranch.True) {template="DISPATCH_RigVMDispatch_ArrayIterator(in Array,out Element,out Index,out Count,out Ratio)"}
   Concat = /Script/RigVM.RigVMFunction_NameConcat(B=_ctrl, A=@DISPATCH_RigVMDispatch_ArrayIterator.Element.Name) {is=struct RigVMUnitNode, template="Concat::Execute(in A,in B,out Result)"}
   > &RigUnit_SetTransform(
    Item=(Type=Bone,Name="spine_01")
    Weight=1
    bInitial=False
    Item.Name=@DISPATCH_RigVMDispatch_ArrayIterator.Element.Name
    Value=&RigUnit_GetTransform(Item=(Type=Control,Name="spine_01_ctrl"), bInitial=False, Item.Name=@Concat.Result).Transform
   )
  ]}
 New Function_ContainedGraph {nodes [
   RigVMFunctionEntryNode()
   > &RigUnit_ParentConstraint(
    AdvancedSettings=(InterpolationType=Average,RotationOrderForFilter=XZY)
    Child=(Type=Bone,Name="base_ponytail_1")
    Parents=((Item=(Type=Bone,Name="clavicle_l"),Weight=0.2),(Item=(Type=Bone,Name="spine_05"),Weight=0.8))
    Weight=1
   )
   > DeltaRotation = RigVMFunctionReferenceNode(
    Rotation Order=ZYX
    child=head
    debugRotation=false
    max=-45
    parent=head_null
    parentType=Null
    premult=0.8
    primAxis=2
    twistAxis=(X=1,Y=0,Z=0)
   )
   > /Script/ControlRig.RigUnit_SetCurveValue(Curve=headRz_up_out, Value=@DeltaRotation.outputWeight) {is=struct RigVMUnitNode}
   > &RigUnit_ModifyTransforms(
    ItemToModify="((Item=(Type=Bone,Name=\"base_hairStrand_E_1\"),Transform=(Rotation=(X=0,Y=-0.3596,Z=0,W=0.9331),Translation=(X=0.2459,Y=0,Z=-0.153),Scale3D=(X=1.1026,Y=1.1026,Z=1.1026))),(Item=(Type=Bone,Name=\"base_hairStrand_I_1\"),Transform=(Rotation=(X=0,Y=-0.4226,Z=0,W= ...(truncated)"
    WeightMinimum=0
    Weight=@DeltaRotation.outputWeight
   )
   > &RigUnit_ModifyTransforms(
    ItemToModify="((Item=(Type=Bone,Name=\"base_hairStrand_C_1\"),Transform=(Rotation=(X=0,Y=0.0436,Z=0,W=0.999),Translation=(X=0,Y=0,Z=0),Scale3D=(X=0.4592,Y=1,Z=1))),(Item=(Type=Bone,Name=\"base_hairStrand_G_1\"),Transform=(Rotation=(X=0,Y=-0.2588,Z=0,W=0.9659),Translation=(X ...(truncated)"
    WeightMinimum=0
    Weight=@DeltaRotation.outputWeight
   )
   > &RigUnit_ModifyTransforms(
    ItemToModify="((Item=(Type=Bone,Name=\"base_hairStrand_J_1\"),Transform=(Rotation=(X=0.0221,Y=-0.3165,Z=0.0662,W=0.946),Translation=(X=0,Y=0,Z=0),Scale3D=(X=0.4592,Y=1,Z=1))),(Item=(Type=Bone,Name=\"dyn_hairStrand_L_1\"),Transform=(Rotation=(X=0.0298,Y=-0.3407,Z=0.0819,W=0. ...(truncated)"
    WeightMinimum=0
    Weight=@DeltaRotation.outputWeight
   )
   > RigVMFunctionReturnNode()
  ]}
 New Function_ContainedGraph {nodes [
   n1 = RigVMFunctionEntryNode()
   RigVMFunction_MathBoolToInteger = &RigVMFunction_MathBoolToInteger(Value=@n1.onlyTwist)
   RigVMFunction_MathQuaternionSwingTwist = &RigVMFunction_MathQuaternionSwingTwist(
    TwistAxis=@n1.twistAxis
    Input=&RigUnit_GetRelativeTransformForItem(
     Child=(Type=Bone,Name="None")
     Parent=(Type=Bone,Name="None")
     bChildInitial=False
     bParentInitial=False
     Child.Name=@n1.child
     Parent.Name=@n1.parent
     Parent.Type=@n1.parentType
    ).RelativeTransform.Rotation
   )
   RigVMFunction_MathQuaternionToEuler = &RigVMFunction_MathQuaternionToEuler(
    Value=RigVMDispatchNode(
     Values=((),())
     Index=@RigVMFunction_MathBoolToInteger.Result
     Values.0=@RigVMFunction_MathQuaternionSwingTwist.Swing
     Values.1=@RigVMFunction_MathQuaternionSwingTwist.Twist
    ) {template="DISPATCH_RigVMDispatch_SelectInt32(in Index,in Values,out Result)"}.Result
    RotationOrder=@n1.Rotation Order
   )
   > DISPATCH_RigVMDispatch_Print = RigVMDispatchNode(ScreenColor=(R=1,G=1,B=1,A=1), ScreenDuration=1.5, Enabled=@n1.debugRotation, Value=@RigVMFunction_MathQuaternionToEuler.Result) {template="DISPATCH_RigVMDispatch_Print(in Prefix,in Value,in Enabled,in ScreenDuration,in ScreenColor)"}
   Clamp = &RigVMFunction_MathIntClamp(Value=@n1.primAxis)
   DISPATCH_RigVMDispatch_SelectInt32 = RigVMDispatchNode(
    Values=(0,0,0)
    Index=@Clamp.Result
    Values.0=@RigVMFunction_MathQuaternionToEuler.Result.X
    Values.1=@RigVMFunction_MathQuaternionToEuler.Result.Y
    Values.2=@RigVMFunction_MathQuaternionToEuler.Result.Z
   ) {template="DISPATCH_RigVMDispatch_SelectInt32(in Index,in Values,out Result)"}
   Remap = &RigVMFunction_MathDoubleRemap(
    TargetMaximum=1
    TargetMinimum=0
    bClamp=False
    Value=@DISPATCH_RigVMDispatch_SelectInt32.Result
    SourceMaximum=@n1.max
    SourceMinimum=@n1.min
   )
   Multiply = &RigVMFunction_MathDoubleMul(A=&RigVMFunction_MathDoubleClamp(Value=@Remap.Result).Result, B=@n1.premult)
   > RigVMFunctionReturnNode(outputWeight=@Multiply.Result)
  ]}
 RigVMFunctionLibrary {nodes [
   DeltaRotation = RigVMCollapseNode(ExecuteContext=(), Rotation Order=ZYX, premult=1, twistAxis=(X=1,Y=0,Z=0))
   controllers = RigVMCollapseNode(ExecuteContext=(), debug=False)
   hair = RigVMCollapseNode(ExecuteContext=())
  ]}
 RigVMModel {nodes [
   RigUnit_BeginExecution = &RigUnit_BeginExecution()
   controllers = RigVMFunctionReferenceNode(debug=false, ExecuteContext=@RigUnit_BeginExecution.ExecutePin)
   > hair = RigVMFunctionReferenceNode()
  ]}
]
hierarchy [
 root {loc=<0 0 0>, rot=<0 0 0 1>}
 pelvis {parent=root, loc=<0 -0.5301 98.0571>, rot=<0.0127 0.707 -0.0127 -0.707>}
 spine_01 {parent=pelvis, loc=<6.1208 -0.9328 0>, rot=<0 0 0.034 -0.9994>}
 spine_02 {parent=spine_01, loc=<6.5997 -0.2095 0>, rot=<0 0 0.0238 -0.9997>}
 spine_03 {parent=spine_02, loc=<6.4914 -0.1237 0>, rot=<0 0 0.0149 -0.9999>}
 spine_04 {parent=spine_03, loc=<6.9598 0.0794 0>, rot=<0 0 -0.0132 -0.9999>}
 spine_05 {parent=spine_04, loc=<6.9987 0 0>, rot=<0 0 -0.0015 -1>}
 neck_01 {parent=spine_05, loc=<11.3138 0.1529 0>, rot=<0 0 -0.2003 -0.9797>}
 neck_02 {parent=neck_01, loc=<4.5627 0 0>, rot=<0 0 0.0504 -0.9987>}
 head {parent=neck_02, loc=<5.3914 0 0>, rot=<0 0 0.115 -0.9934>}
 base_earring_r_1 {parent=head, loc=<0.4247 0.1363 6.4376>, rot=<-0.0327 0.3165 0.948 -0.0092>}
 dyn_earring_r_1 {parent=base_earring_r_1, loc=<0 0 0>, rot=<0 0 0 1>}
 dyn_earring_r_2 {parent=dyn_earring_r_1, loc=<1.7275 0 0>, rot=<0.684 0.0103 0.0044 -0.7294>}
 base_earring_l_1 {parent=head, loc=<0.4248 0.1363 -6.4376>, rot=<0.0092 -0.948 -0.3165 0.0327>}
 dyn_earring_l_1 {parent=base_earring_l_1, loc=<0 0 0>, rot=<0 0 0 1>}
 dyn_earring_l_2 {parent=dyn_earring_l_1, loc=<1.7279 0 0>, rot=<-0.684 0.0103 -0.0044 -0.7294>}
 base_hairStrand_A_1 {parent=head, loc=<1.2619 -10.1797 -4.4158>, rot=<-0.1805 0.4936 0.8506 -0.0194>}
 dyn_hairStrand_A_1 {parent=base_hairStrand_A_1, loc=<0 0 0>, rot=<0 0 0 1>}
 base_hairStrand_B_1 {parent=head, loc=<13.1297 8.2723 2.1981>, rot=<0.7269 -0.0735 0.6681 -0.1411>}
 dyn_hairStrand_B_1 {parent=base_hairStrand_B_1, loc=<0 0 0>, rot=<0 0 0 1>}
 base_hairStrand_C_1 {parent=head, loc=<5.2635 0.1839 8.7411>, rot=<0.0087 -0.9858 0.0733 0.1506>}
 dyn_hairStrand_C_1 {parent=base_hairStrand_C_1, loc=<0 0 0>, rot=<0 0 0 1>}
 base_hairStrand_D_1 {parent=head, loc=<7.0703 -7.5713 7.1563>, rot=<0.1075 -0.9088 -0.3818 0.1295>}
 dyn_hairStrand_D_1 {parent=base_hairStrand_D_1, loc=<0 0 0>, rot=<0 0 0 1>}
 base_hairStrand_E_1 {parent=head, loc=<7.153 -9.855 4.0629>, rot=<0.1682 -0.8175 -0.5465 0.0692>}
 dyn_hairStrand_E_1 {parent=base_hairStrand_E_1, loc=<0 0 0>, rot=<0 0 0 1>}
 base_hairStrand_F_1 {parent=head, loc=<7.739 7.969 5.6205>, rot=<0.1224 -0.9745 0.1569 0.1038>}
 dyn_hairStrand_F_1 {parent=base_hairStrand_F_1, loc=<0 0 0>, rot=<0 0 0 1>}
 base_hairStrand_G_1 {parent=head, loc=<7.7198 5.0695 6.5374>, rot=<0.1186 -0.9889 0.0867 0.02>}
 dyn_hairStrand_G_1 {parent=base_hairStrand_G_1, loc=<0 0 0>, rot=<0 0 0 1>}
 dyn_hairStrand_G_2 {parent=dyn_hairStrand_G_1, loc=<8.9156 0 0>, rot=<-0.0008 0.0229 -0.0344 -0.9991>}
 base_hairStrand_H_1 {parent=head, loc=<1.6456 -3.2126 8.7668>, rot=<0.0123 -0.9699 -0.2412 0.0297>}
 dyn_hairStrand_H_1 {parent=base_hairStrand_H_1, loc=<0 0 0>, rot=<0 0 0 1>}
 base_hairStrand_I_1 {parent=head, loc=<6.0957 -4.8657 8.8344>, rot=<0.1394 -0.8669 -0.2561 0.4042>}
 dyn_hairStrand_I_1 {parent=base_hairStrand_I_1, loc=<0 0 0>, rot=<0 0 0 1>}
 base_hairStrand_J_1 {parent=head, loc=<6.5451 -8.0922 -6.5855>, rot=<-0.4066 0.4634 0.7597 -0.207>}
 dyn_hairStrand_J_1 {parent=base_hairStrand_J_1, loc=<0 0 0>, rot=<0 0 0 1>}
 base_hairStrand_K_1 {parent=head, loc=<0.9271 -5.343 -8.8476>, rot=<-0.2393 0.464 0.8526 -0.021>}
 dyn_hairStrand_K_1 {parent=base_hairStrand_K_1, loc=<0 0 0>, rot=<0 0 0 1>}
 base_hairStrand_L_1 {parent=head, loc=<4.7083 -10.6963 -1.4065>, rot=<-0.3589 0.5648 0.7425 -0.0311>}
 dyn_hairStrand_L_1 {parent=base_hairStrand_L_1, loc=<0 0 0>, rot=<0 0 0 1>}
 base_hairStrand_M_1 {parent=head, loc=<4.6617 -1.6103 -8.731>, rot=<-0.5681 0.3659 0.6879 -0.2648>}
 dyn_hairStrand_M_1 {parent=base_hairStrand_M_1, loc=<0 0 0>, rot=<0 0 0 1>}
 base_hairStrand_N_1 {parent=head, loc=<7.2573 3.0962 -7.2608>, rot=<-0.1287 -0.2111 0.969 0.0006>}
 dyn_hairStrand_N_1 {parent=base_hairStrand_N_1, loc=<0 0 0>, rot=<0 0 0 1>}
 base_hairStrand_O_1 {parent=head, loc=<11.4381 -5.9632 -5.4755>, rot=<-0.591 0.4116 0.5951 -0.3567>}
 dyn_hairStrand_O_1 {parent=base_hairStrand_O_1, loc=<0 0 0>, rot=<0 0 0 1>}
 base_hairStrand_P_1 {parent=head, loc=<9.8595 -0.4863 -9.601>, rot=<0.6895 -0.2502 -0.5547 0.393>}
 dyn_hairStrand_P_1 {parent=base_hairStrand_P_1, loc=<0 0 0>, rot=<0 0 0 1>}
 base_hairStrand_Q_1 {parent=head, loc=<8.4856 7.9094 -5.2005>, rot=<-0.1075 -0.2327 0.9518 -0.1684>}
 dyn_hairStrand_Q_1 {parent=base_hairStrand_Q_1, loc=<0 0 0>, rot=<0 0 0 1>}
 dyn_hairStrand_Q_2 {parent=dyn_hairStrand_Q_1, loc=<6.1095 0 0>, rot=<0.0055 -0.0332 -0.1644 -0.9858>}
 base_hairStrand_R_1 {parent=head, loc=<7.259 -3.3803 -8.5203>, rot=<0.6603 -0.3386 -0.5295 0.4112>}
 dyn_hairStrand_R_1 {parent=base_hairStrand_R_1, loc=<0 0 0>, rot=<0 0 0 1>}
 base_hairStrand_S_1 {parent=head, loc=<11.9832 -3.2467 -7.6568>, rot=<0.7588 -0.3657 -0.3559 0.4047>}
 dyn_hairStrand_S_1 {parent=base_hairStrand_S_1, loc=<0 0 0>, rot=<0 0 0 1>}
 base_hairStrand_T_1 {parent=head, loc=<10.0149 4.6774 -6.5633>, rot=<-0.6648 0.1292 0.6808 -0.2791>}
 dyn_hairStrand_T_1 {parent=base_hairStrand_T_1, loc=<0 0 0>, rot=<0 0 0 1>}
 base_hairStrand_U_1 {parent=head, loc=<15.4189 2.9142 1.9242>, rot=<0.69 -0.3107 0.479 -0.4449>}
 dyn_hairStrand_U_1 {parent=base_hairStrand_U_1, loc=<0 0 0>, rot=<0 0 0 1>}
 base_hairStrand_V_1 {parent=head, loc=<7.4571 1.2658 -8.382>, rot=<-0.3529 -0.3806 0.6607 -0.5423>}
 dyn_hairStrand_V_1 {parent=base_hairStrand_V_1, loc=<0 0 0>, rot=<0 0 0 1>}
 FACIAL_C_FacialRoot {parent=head, loc=<1.9918 0.2205 0.0005>, rot=<0.5021 -0.4979 -0.5021 -0.4979>}
 FACIAL_L_Ear {parent=FACIAL_C_FacialRoot, loc=<5.9627 -0.2512 0.2089>, rot=<-0.0262 -0.0007 -0.0262 -0.9993>}
 FACIAL_R_Ear {parent=FACIAL_C_FacialRoot, loc=<-5.9627 -0.2512 0.2089>, rot=<-0.0262 0.0007 0.0262 -0.9993>}
 base_scalp_back_c {parent=head, loc=<1.1516 -10.1137 5.0854>, rot=<0.2985 0 0 -0.9544>}
 kin_scalp_back_c {parent=base_scalp_back_c, loc=<0 0 0>, rot=<0 0 0 1>}
 base_hairStrand_CH_1 {parent=head, loc=<0.6564 -3.1855 -8.6743>, rot=<-0.0655 0.4692 0.8702 0.1355>}
 dyn_hairStrand_CH_1 {parent=base_hairStrand_CH_1, loc=<0 0 0>, rot=<0 0 0 1>}
 base_hairStrand_W_1 {parent=head, loc=<6.7852 -8.8418 -3.5983>, rot=<-0.4495 0.5258 0.6809 -0.2405>}
 dyn_hairStrand_W_1 {parent=base_hairStrand_W_1, loc=<0 0 0>, rot=<0 0 0 1>}
 base_hairStrand_X_1 {parent=head, loc=<10.4316 -3.4002 7.457>, rot=<-0.3345 0.6922 0.3166 -0.5556>}
 dyn_hairStrand_X_1 {parent=base_hairStrand_X_1, loc=<0 0 0>, rot=<0 0 0 1>}
 base_hairStrand_Y_1 {parent=head, loc=<1.3269 -8.8766 -7.0352>, rot=<-0.287 0.4871 0.8208 -0.0819>}
 dyn_hairStrand_Y_1 {parent=base_hairStrand_Y_1, loc=<0 0 0>, rot=<0 0 0 1>}
 base_hairStrand_Z_1 {parent=head, loc=<3.6222 -8.0387 -7.794>, rot=<-0.4621 0.3338 0.8213 -0.0218>}
 dyn_hairStrand_Z_1 {parent=base_hairStrand_Z_1, loc=<0 0 0>, rot=<0 0 0 1>}
 coll_earring_ponytail {parent=neck_01, loc=<4.8753 -0.4902 -6.8231>, rot=<-0.5107 -0.4891 -0.329 -0.6259>}
 clavicle_l {parent=spine_05, loc=<5.4831 -0.1053 -3.2946>, rot=<0.0465 -0.7289 -0.009 -0.6829>}
 upperarm_l {parent=clavicle_l, loc=<9.8658 -1.2136 -0.4105>, rot=<0.0164 -0.3245 -0.0269 -0.9454>}
 clavicle_r {parent=spine_05, loc=<5.4831 -0.1053 3.2946>, rot=<0.7289 0.0465 -0.6829 0.009>}
 upperarm_r {parent=clavicle_r, loc=<-9.8658 1.2136 0.4105>, rot=<0.0164 -0.3245 -0.0269 -0.9454>}
 base_ponytail_1 {parent=spine_05, loc=<13.9988 -1.5249 -6.7111>, rot=<-0.4565 -0.8879 0.0527 -0.0196>}
 dyn_ponytail_1 {parent=base_ponytail_1, loc=<0 0 0>, rot=<0 0 0 1>}
 dyn_ponytail_2 {parent=dyn_ponytail_1, loc=<7.3259 0 0>, rot=<0 -0.1485 0.1246 -0.981>}
 dyn_ponytail_3 {parent=dyn_ponytail_2, loc=<7.3653 0 0>, rot=<-0.0538 -0.0148 0.0637 -0.9964>}
 dyn_ponytail_4 {parent=dyn_ponytail_3, loc=<7.3691 0 0>, rot=<0.0763 0.07 0.0715 -0.9921>}
 dyn_ponytail_5 {parent=dyn_ponytail_4, loc=<5.2479 0 0>, rot=<0.0198 0.0645 0.3217 -0.9444>}
 dyn_ponytail_A_5 {parent=dyn_ponytail_4, loc=<3.0598 0.0864 0.9895>, rot=<0.0072 0.2422 0.257 -0.9355>}
 dyn_ponytail_B_5 {parent=dyn_ponytail_4, loc=<1.5784 0.5708 -0.7423>, rot=<-0.2981 -0.3154 -0.1357 -0.8907>}
 head_null {parent=spine_05, type=Null, loc=<20.6573 3.5489 0>, rot=<0 0 -0.036 -0.9994>}
 base_neckCollar_l_1 {parent=spine_05, loc=<7.162 2.44 -12.7>, rot=<0.9221 0.2961 -0.2487 -0.0143>}
 dyn_neckCollar_l_1 {parent=base_neckCollar_l_1, loc=<0 0 0>, rot=<0 0 0 1>}
 base_neckCollar_l_2 {parent=spine_05, loc=<8.332 -4.1203 -11.4499>, rot=<0.9554 0.1016 -0.1662 0.222>}
 dyn_neckCollar_l_2 {parent=base_neckCollar_l_2, loc=<0 0 0>, rot=<0 0 0 1>}
 base_neckCollar_l_3 {parent=spine_05, loc=<8.4428 -9.5327 -6.282>, rot=<0.8557 -0.0526 -0.1305 0.498>}
 dyn_neckCollar_l_3 {parent=base_neckCollar_l_3, loc=<0 0 0>, rot=<0 0 0 1>}
 base_neckCollar_c_1 {parent=spine_05, loc=<8.8392 -11.3332 0.865>, rot=<-0.7036 0.0706 0.0706 -0.7036>}
 dyn_neckCollar_c_1 {parent=base_neckCollar_c_1, loc=<0 0 0>, rot=<0 0 0 1>}
 base_neckCollar_r_1 {parent=spine_05, loc=<7.1615 2.4399 12.7>, rot=<0.0143 0.2487 -0.2961 -0.9221>}
 dyn_neckCollar_r_1 {parent=base_neckCollar_r_1, loc=<0 0 0>, rot=<0 0 0 1>}
 base_neckCollar_r_2 {parent=spine_05, loc=<8.332 -4.1203 11.4499>, rot=<-0.222 0.1662 -0.1016 -0.9554>}
 dyn_neckCollar_r_2 {parent=base_neckCollar_r_2, loc=<0 0 0>, rot=<0 0 0 1>}
 base_neckCollar_r_3 {parent=spine_05, loc=<8.4431 -9.5327 6.282>, rot=<-0.498 0.1305 0.0526 -0.8557>}
 dyn_neckCollar_r_3 {parent=base_neckCollar_r_3, loc=<0 0 0>, rot=<0 0 0 1>}
 neck_01_ctrl {controlType=EULER_TRANSFORM, type=Control, loc=<0 0 0>, rot=<0 0 0 1>}
 neck_02_ctrl {controlType=EULER_TRANSFORM, parent=neck_01_ctrl, type=Control, loc=<0 0 0>, rot=<0 0 0 1>}
 head_ctrl {controlType=EULER_TRANSFORM, parent=neck_02_ctrl, type=Control, loc=<0 0 0>, rot=<0 0 0 1>}
 clavicle_r_ctrl {controlType=EULER_TRANSFORM, type=Control, loc=<0 0 0>, rot=<0 0 0 1>}
 clavicle_l_ctrl {controlType=EULER_TRANSFORM, type=Control, loc=<0 0 0>, rot=<0 0 0 1>}
 headRz_up_out {type=Curve, loc=<0 0 0>, rot=<0 0 0 1>}
]
