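# Physics asset for the face accessory mesh of the female medium PeonyBellow player cosmetic, setting up secondary motion for ponytail, hair strand and earrings.
# Capsule chains run from base bodies through dyn bodies with narrow swing limits and zero-degree twist, with the ponytail branching into three tips.
# Head, neck, clavicle, upper spine and a dedicated earring-ponytail capsule provide collision volumes.
# Constraints among these body parts and to chain bases are fully free in rotation.
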PhysicsAsset /BRCosmetics/Characters/Player/Female/Medium/Bodies/F_MED_PeonyBellow/Meshes/Parts/F_MED_PeonyBellow_FaceAcc_Physics
bodies {
 base_earring_l_1 [{shapeType=Capsule, length=1, radius=0.2, shapeName=base_ponytail_1_capsule}]
 base_earring_r_1 [{shapeType=Capsule, length=1, radius=0.2, shapeName=base_ponytail_1_capsule}]
 base_hairStrand_G_1 [{shapeType=Capsule, length=1, radius=0.5, shapeName=base_hairStrand_G_1_capsule}]
 base_ponytail_1 [{shapeType=Capsule, length=1, radius=0.5, shapeName=base_ponytail_1_capsule}]
 clavicle_l [
  {shapeType=Capsule, length=6.6899, radius=6.5516, shapeName=clavicle_l_capsule, center=<4.7885 0 -2.9662>, rotation=<rot 90 0 0>}
  {
   shapeType=Capsule
   length=7.3303
   radius=5.9112
   shapeName=clavicle_l_capsule
   center=<12.6606 5.1001 4.2416>
   rotation=<rot 71.2607 99.9403 -170.4938>
  }
 ]
 clavicle_r [{shapeType=Capsule, length=6.6899, radius=6.5516, shapeName=clavicle_r_capsule, center=<-4.7885 0 2.9662>, rotation=<rot 90 0 180>}]
 coll_earring_ponytail [{
   shapeType=Capsule
   length=4.0193
   radius=2.3897
   shapeName=coll_earring_ponytail_capsule
   center=<-1.9666 -0.2233 -0.4461>
   rotation=<rot -5.8439 115.4445 -166.0677>
  }]
 dyn_earring_l_1 [{
   shapeType=Capsule
   length=1.7034
   radius=0.5
   shapeName=dyn_earring_l_1_capsule
   center=<1.1233 0.0479 0.0214>
   rotation=<rot -1.6136 176.3402 -89.9571>
  }]
 dyn_earring_l_2 [{
   shapeType=Capsule
   length=2.5464
   radius=0.6195
   shapeName=dyn_earring_l_2_capsule
   center=<3.2436 0.0004 0>
   rotation=<rot 0 0.0185 -87.7141>
  }]
 dyn_earring_r_1 [{
   shapeType=Capsule
   length=1.7034
   radius=0.5
   shapeName=dyn_earring_l_1_capsule
   center=<1.1233 0.0479 0.0214>
   rotation=<rot -1.6136 176.3402 -89.9571>
  }]
 dyn_earring_r_2 [{
   shapeType=Capsule
   length=2.1868
   radius=0.8793
   shapeName=dyn_earring_l_2_capsule
   center=<2.3096 -1.1035 0.044>
   rotation=<rot -2.1905 106.5057 -90.6527>
  }]
 dyn_hairStrand_G_1 [{
   shapeType=Capsule
   length=5.7056
   radius=1.6843
   shapeName=dyn_hairStrand_G_2_capsule
   center=<3.8305 0 -0.4995>
   rotation=<rot 82.9314 -180 180>
  }]
 dyn_hairStrand_G_2 [{
   shapeType=Capsule
   length=3.8031
   radius=1.1193
   shapeName=dyn_hairStrand_G_2_capsule
   center=<2.0849 0.0172 0.1793>
   rotation=<rot -0.1706 -89.6723 -88.637>
  }]
 dyn_ponytail_1 [{
   shapeType=Capsule
   length=1.9953
   radius=2.7654
   shapeName=dyn_ponytail_1_capsule
   center=<4.2547 0.6542 0.3666>
   rotation=<rot 0 -99.4351 -82.8966>
  }]
 dyn_ponytail_2 [{shapeType=Capsule, length=3.9864, radius=2.9921, shapeName=dyn_ponytail_1_capsule, center=<3.5899 0.8602 0>, rotation=<rot 0 -90 -90>}]
 dyn_ponytail_3 [{
   shapeType=Capsule
   length=5.0447
   radius=1.912
   shapeName=dyn_ponytail_3_capsule
   center=<3.0989 0.6591 0.2789>
   rotation=<rot -0.0001 95.7842 -90.0001>
  }]
 dyn_ponytail_4 [{
   shapeType=Capsule
   length=3.0969
   radius=2.2631
   shapeName=dyn_ponytail_4_capsule
   center=<2.8662 0.998 0.0266>
   rotation=<rot 0.1825 -98.4398 -88.483>
  }]
 dyn_ponytail_5 [{
   shapeType=Capsule
   length=2.1989
   radius=1.3386
   shapeName=dyn_ponytail_5_capsule
   center=<2.7425 -0.2202 0.2623>
   rotation=<rot -5.405 83.4845 -67.3311>
  }]
 dyn_ponytail_A_5 [{
   shapeType=Capsule
   length=1.8624
   radius=1.0733
   shapeName=dyn_ponytail_A_5_capsule
   center=<3.2346 0.1344 0.2433>
   rotation=<rot -1.8918 78.526 -68.0828>
  }]
 dyn_ponytail_B_5 [{
   shapeType=Capsule
   length=1.7964
   radius=1.0559
   shapeName=dyn_ponytail_B_5_capsule
   center=<3.2156 -0.1136 -0.0411>
   rotation=<rot -3.1563 -75.2089 -69.6698>
  }]
 head [{shapeType=Capsule, length=11.8133, radius=5.6957, shapeName=head_capsule, center=<5.9514 2.4581 0>, rotation=<rot -90 180 -180>}]
 neck_01 [{shapeType=Capsule, length=4.2323, radius=4.192, shapeName=neck_01_capsule, center=<3.0292 0 0>, rotation=<rot 90 0 0>}]
 spine_03 [{shapeType=Capsule, length=6.8809, radius=8.3895, shapeName=spine_05_capsule, center=<0 3.26 0>}]
 spine_04 [
  {shapeType=Capsule, length=4.8555, radius=10.5202, shapeName=spine_05_capsule, center=<0 3.1292 0>}
  {
   shapeType=Capsule
   length=8.4697
   radius=6.9061
   shapeName=spine_05_capsule
   center=<1.1965 3.1158 -18.2805>
   rotation=<rot 0 -0.0004 82.1489>
  }
  {
   shapeType=Capsule
   length=16.4502
   radius=7.9182
   shapeName=spine_05_capsule
   center=<6.1425 5.1165 17.2649>
   rotation=<rot 0 -0.0004 82.1489>
  }
 ]
 spine_05 [
  {shapeType=Capsule, length=9.2601, radius=8.3895, shapeName=spine_05_capsule, center=<0 -0.6777 0>}
  {shapeType=Capsule, length=12.7964, radius=5.782, shapeName=spine_05_capsule, center=<0 2.2839 -17.5571>, rotation=<rot 0 0 80.4254>}
 ]
}
constraints [
 {bone1Name=dyn_ponytail_B_5, bone2Name=dyn_ponytail_4, swing1LimitDegrees=15, swing2LimitDegrees=15, twistLimitDegrees=0}
 {bone1Name=dyn_ponytail_1, bone2Name=base_ponytail_1, swing1LimitDegrees=10, swing2LimitDegrees=10, twistLimitDegrees=0}
 {bone1Name=dyn_ponytail_2, bone2Name=dyn_ponytail_1, swing1LimitDegrees=12, swing2LimitDegrees=8, twistLimitDegrees=0}
 {bone1Name=dyn_ponytail_3, bone2Name=dyn_ponytail_2, swing1LimitDegrees=15, swing2LimitDegrees=10, twistLimitDegrees=0}
 {bone1Name=dyn_ponytail_4, bone2Name=dyn_ponytail_3, swing1LimitDegrees=15, swing2LimitDegrees=20, twistLimitDegrees=0}
 {bone1Name=dyn_ponytail_5, bone2Name=dyn_ponytail_4, swing1LimitDegrees=15, swing2LimitDegrees=15, twistLimitDegrees=0}
 {bone1Name=dyn_ponytail_A_5, bone2Name=dyn_ponytail_4, swing1LimitDegrees=15, swing2LimitDegrees=15, twistLimitDegrees=0}
 {
  bone1Name=clavicle_l
  bone2Name=spine_05
  swing1LimitDegrees=45
  swing1Motion=Free
  swing2LimitDegrees=45
  swing2Motion=Free
  twistLimitDegrees=45
  twistMotion=Free
 }
 {
  bone1Name=base_ponytail_1
  bone2Name=spine_05
  swing1LimitDegrees=45
  swing1Motion=Free
  swing2LimitDegrees=45
  swing2Motion=Free
  twistLimitDegrees=45
  twistMotion=Free
 }
 {
  bone1Name=neck_01
  bone2Name=spine_05
  swing1LimitDegrees=45
  swing1Motion=Free
  swing2LimitDegrees=45
  swing2Motion=Free
  twistLimitDegrees=45
  twistMotion=Free
 }
 {
  bone1Name=spine_05
  bone2Name=spine_04
  swing1LimitDegrees=45
  swing1Motion=Free
  swing2LimitDegrees=45
  swing2Motion=Free
  twistLimitDegrees=45
  twistMotion=Free
 }
 {
  bone1Name=spine_04
  bone2Name=spine_03
  swing1LimitDegrees=45
  swing1Motion=Free
  swing2LimitDegrees=45
  swing2Motion=Free
  twistLimitDegrees=45
  twistMotion=Free
 }
 {bone1Name=dyn_earring_l_2, bone2Name=dyn_earring_l_1, swing1LimitDegrees=5, swing2LimitDegrees=20, twistLimitDegrees=0}
 {bone1Name=dyn_earring_l_1, bone2Name=base_earring_l_1, swing1LimitDegrees=20, swing2LimitDegrees=10, twistLimitDegrees=0}
 {
  bone1Name=base_earring_l_1
  bone2Name=head
  swing1LimitDegrees=45
  swing1Motion=Free
  swing2LimitDegrees=45
  swing2Motion=Free
  twistLimitDegrees=45
  twistMotion=Free
 }
 {bone1Name=dyn_earring_r_2, bone2Name=dyn_earring_r_1, swing1LimitDegrees=5, swing2LimitDegrees=20, twistLimitDegrees=0}
 {bone1Name=dyn_earring_r_1, bone2Name=base_earring_r_1, swing1LimitDegrees=20, swing2LimitDegrees=10, twistLimitDegrees=0}
 {
  bone1Name=base_earring_r_1
  bone2Name=head
  swing1LimitDegrees=45
  swing1Motion=Free
  swing2LimitDegrees=45
  swing2Motion=Free
  twistLimitDegrees=45
  twistMotion=Free
 }
 {
  bone1Name=head
  bone2Name=neck_01
  swing1LimitDegrees=45
  swing1Motion=Free
  swing2LimitDegrees=45
  swing2Motion=Free
  twistLimitDegrees=45
  twistMotion=Free
 }
 {
  bone1Name=coll_earring_ponytail
  bone2Name=neck_01
  swing1LimitDegrees=45
  swing1Motion=Free
  swing2LimitDegrees=45
  swing2Motion=Free
  twistLimitDegrees=45
  twistMotion=Free
 }
 {bone1Name=dyn_hairStrand_G_2, bone2Name=dyn_hairStrand_G_1, swing1LimitDegrees=20, swing2LimitDegrees=20, twistLimitDegrees=0}
 {
  bone1Name=base_hairStrand_G_1
  bone2Name=head
  swing1LimitDegrees=45
  swing1Motion=Free
  swing2LimitDegrees=45
  swing2Motion=Free
  twistLimitDegrees=45
  twistMotion=Free
 }
 {bone1Name=dyn_hairStrand_G_1, bone2Name=base_hairStrand_G_1, swing1LimitDegrees=10, swing2LimitDegrees=25, twistLimitDegrees=0}
]
