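# Post-process style AnimBlueprint for Peony Bellow's face accessory part, driving hair, earrings and collar.
# Parent and body poses are blended through a Control Rig, then ear and collar transforms sampled from head and body meshes are applied.
# Constraints and additive offsets prepare earring and ponytail bones before chained AnimDynamics hair strands and a RigidBody simulation.
# Event graph updates gravity, simulation-space settings and skydive, glider and curve-driven weights.
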AnimBlueprint /BRCosmetics/Characters/Player/Female/Medium/Bodies/F_MED_PeonyBellow/Meshes/Parts/F_MED_PeonyBellow_FaceAcc_AnimBP
target_skeleton=&/F_MED_PeonyBellow_FaceAcc_Skeleton
graphs [
 AnimGraph {nodes [
   Output Pose(Result=Use cached pose 'dynamicsCP'()) {layerGroup=None}
   n1 = Get BodyMesh() -> (BodyMesh:Skeletal Mesh Component Object Reference)
   n2 = Control Rig(Source=Layered blend per bone(
     BasePose=Copy Pose From Mesh() {bUseAttachedParent=true}
     BlendPoses_0=Copy Pose From Mesh(SourceMeshComponent=@n1) {bUseAttachedParent=true}
    )) {
    alphaBoolBlend {blendOption=Linear}
    controlRigAssetReference {blueprintRigClass=&/F_MED_PeonyBellow_FaceAcc_CtrlRig.F_MED_PeonyBellow_FaceAcc_CtrlRig_C}
   }
   Save cached pose 'controlRigCP'(Pose=@n2) {cacheName=controlRigCP}
   n3 = Anim Dynamics - Bone: dyn_hairStrand_CH_1!(ComponentPose=Local To Component(LocalPose=Use cached pose 'presimFixesCP'()), Alpha=1, GravityOverride:Vector="0,0,-980") -> (Pose) {
    angularDampingOverride=1
    angularSpringConstant=1200
    bAngularSpring=true
    bOverrideAngularDamping=true
    bUseGravityOverride=true
    gravityScale=0.35
    simulationSpace=RootRelative
    alphaBoolBlend {blendOption=Linear}
    boundBone {boneName=dyn_hairStrand_CH_1}
    componentAppliedLinearAccClamp=<5000 5000 5000>
    componentLinearAccScale=<0.3 0.3 0.3>
    componentLinearVelScale=<0.3 0.3 0.3>
    gravityOverride {z=-980}
    physicsBodyDefinitions [{
      boundBone {boneName=dyn_hairStrand_CH_1}
      boxExtents=<10 10 10>
      constraintSetup {angularLimitsMax=<0 15 15>, angularLimitsMin=<0 -15 -15>, angularTarget=<1 0 0>}
      localJointOffset=<1 0 0>
     }]
   }
   n4 = Anim Dynamics - Bone: dyn_hairStrand_M_1!(ComponentPose=@n3, Alpha=1, GravityOverride:Vector="0,0,-980") -> (Pose) {
    angularDampingOverride=1
    angularSpringConstant=1200
    bAngularSpring=true
    bOverrideAngularDamping=true
    bUseGravityOverride=true
    gravityScale=0.35
    simulationSpace=RootRelative
    alphaBoolBlend {blendOption=Linear}
    boundBone {boneName=dyn_hairStrand_M_1}
    componentAppliedLinearAccClamp=<5000 5000 5000>
    componentLinearAccScale=<0.3 0.3 0.3>
    componentLinearVelScale=<0.3 0.3 0.3>
    gravityOverride {z=-980}
    physicsBodyDefinitions [{
      boundBone {boneName=dyn_hairStrand_M_1}
      boxExtents=<10 10 10>
      constraintSetup {angularLimitsMax=<0 10 10>, angularLimitsMin=<0 -10 -5>, angularTarget=<1 0 0>}
      localJointOffset=<1 0 0>
     }]
   }
   n5 = Anim Dynamics - Bone: dyn_hairStrand_N_1!(ComponentPose=@n4, Alpha=1, GravityOverride:Vector="0,0,-980") -> (Pose) {
    angularDampingOverride=1
    angularSpringConstant=1200
    bAngularSpring=true
    bOverrideAngularDamping=true
    bUseGravityOverride=true
    gravityScale=0.35
    simulationSpace=RootRelative
    alphaBoolBlend {blendOption=Linear}
    boundBone {boneName=dyn_hairStrand_N_1}
    componentAppliedLinearAccClamp=<5000 5000 5000>
    componentLinearAccScale=<0.3 0.3 0.3>
    componentLinearVelScale=<0.3 0.3 0.3>
    gravityOverride {z=-980}
    physicsBodyDefinitions [{
      boundBone {boneName=dyn_hairStrand_N_1}
      boxExtents=<10 10 10>
      constraintSetup {angularLimitsMax=<0 15 15>, angularLimitsMin=<0 -15 -15>, angularTarget=<1 0 0>}
      localJointOffset=<1 0 0>
     }]
   }
   n6 = Anim Dynamics - Bone: dyn_hairStrand_O_1!(ComponentPose=@n5, Alpha=1, GravityOverride:Vector="0,0,-980") -> (Pose) {
    angularDampingOverride=1
    angularSpringConstant=1200
    bAngularSpring=true
    bOverrideAngularDamping=true
    bUseGravityOverride=true
    gravityScale=0.35
    simulationSpace=RootRelative
    alphaBoolBlend {blendOption=Linear}
    boundBone {boneName=dyn_hairStrand_O_1}
    componentAppliedLinearAccClamp=<5000 5000 5000>
    componentLinearAccScale=<0.3 0.3 0.3>
    componentLinearVelScale=<0.3 0.3 0.3>
    gravityOverride {z=-980}
    physicsBodyDefinitions [{
      boundBone {boneName=dyn_hairStrand_O_1}
      boxExtents=<10 10 10>
      constraintSetup {angularLimitsMax=<0 15 15>, angularLimitsMin=<0 -15 -15>, angularTarget=<1 0 0>}
      localJointOffset=<1 0 0>
     }]
   }
   n7 = Anim Dynamics - Bone: dyn_hairStrand_P_1!(ComponentPose=@n6, Alpha=1, GravityOverride:Vector="0,0,-980") -> (Pose) {
    angularDampingOverride=1
    angularSpringConstant=1200
    bAngularSpring=true
    bOverrideAngularDamping=true
    bUseGravityOverride=true
    gravityScale=0.35
    simulationSpace=RootRelative
    alphaBoolBlend {blendOption=Linear}
    boundBone {boneName=dyn_hairStrand_P_1}
    componentAppliedLinearAccClamp=<5000 5000 5000>
    componentLinearAccScale=<0.3 0.3 0.3>
    componentLinearVelScale=<0.3 0.3 0.3>
    gravityOverride {z=-980}
    physicsBodyDefinitions [{
      boundBone {boneName=dyn_hairStrand_P_1}
      boxExtents=<10 10 10>
      constraintSetup {angularLimitsMax=<0 15 10>, angularLimitsMin=<0 0 0>, angularTarget=<1 0 0>}
      localJointOffset=<1 0 0>
     }]
   }
   n8 = Anim Dynamics - Bone: dyn_hairStrand_Q_1!(ComponentPose=@n7, Alpha=1, GravityOverride:Vector="0,0,-980") -> (Pose) {
    angularDampingOverride=1
    angularSpringConstant=1200
    bAngularSpring=true
    bOverrideAngularDamping=true
    bUseGravityOverride=true
    gravityScale=0.35
    simulationSpace=RootRelative
    alphaBoolBlend {blendOption=Linear}
    boundBone {boneName=dyn_hairStrand_Q_1}
    componentAppliedLinearAccClamp=<5000 5000 5000>
    componentLinearAccScale=<0.3 0.3 0.3>
    componentLinearVelScale=<0.3 0.3 0.3>
    gravityOverride {z=-980}
    physicsBodyDefinitions [{
      boundBone {boneName=dyn_hairStrand_Q_1}
      boxExtents=<10 10 10>
      constraintSetup {angularLimitsMax=<0 10 15>, angularLimitsMin=<0 -5 -5>, angularTarget=<1 0 0>}
      localJointOffset=<1 0 0>
     }]
   }
   n9 = Anim Dynamics - Bone: dyn_hairStrand_Q_2!(ComponentPose=@n8, Alpha=1, GravityOverride:Vector="0,0,-980") -> (Pose) {
    angularDampingOverride=1
    angularSpringConstant=1200
    bAngularSpring=true
    bOverrideAngularDamping=true
    bUseGravityOverride=true
    gravityScale=0.35
    simulationSpace=RootRelative
    alphaBoolBlend {blendOption=Linear}
    boundBone {boneName=dyn_hairStrand_Q_2}
    componentAppliedLinearAccClamp=<5000 5000 5000>
    componentLinearAccScale=<0.3 0.3 0.3>
    componentLinearVelScale=<0.3 0.3 0.3>
    gravityOverride {z=-980}
    physicsBodyDefinitions [{
      boundBone {boneName=dyn_hairStrand_Q_2}
      boxExtents=<10 10 10>
      constraintSetup {angularLimitsMax=<0 10 7>, angularLimitsMin=<0 -5 -7>, angularTarget=<1 0 0>}
      localJointOffset=<1 0 0>
     }]
   }
   n10 = Anim Dynamics - Bone: dyn_hairStrand_R_1!(ComponentPose=@n9, Alpha=1, GravityOverride:Vector="0,0,-980") -> (Pose) {
    angularDampingOverride=1
    angularSpringConstant=1200
    bAngularSpring=true
    bOverrideAngularDamping=true
    bUseGravityOverride=true
    gravityScale=0.35
    simulationSpace=RootRelative
    alphaBoolBlend {blendOption=Linear}
    boundBone {boneName=dyn_hairStrand_R_1}
    componentAppliedLinearAccClamp=<5000 5000 5000>
    componentLinearAccScale=<0.3 0.3 0.3>
    componentLinearVelScale=<0.3 0.3 0.3>
    gravityOverride {z=-980}
    physicsBodyDefinitions [{
      boundBone {boneName=dyn_hairStrand_R_1}
      boxExtents=<10 10 10>
      constraintSetup {angularLimitsMax=<0 15 15>, angularLimitsMin=<0 -15 -15>, angularTarget=<1 0 0>}
      localJointOffset=<1 0 0>
     }]
   }
   n11 = Anim Dynamics - Bone: dyn_hairStrand_S_1!(ComponentPose=@n10, Alpha=1, GravityOverride:Vector="0,0,-980") -> (Pose) {
    angularDampingOverride=1
    angularSpringConstant=1200
    bAngularSpring=true
    bOverrideAngularDamping=true
    bUseGravityOverride=true
    gravityScale=0.35
    simulationSpace=RootRelative
    alphaBoolBlend {blendOption=Linear}
    boundBone {boneName=dyn_hairStrand_S_1}
    componentAppliedLinearAccClamp=<5000 5000 5000>
    componentLinearAccScale=<0.3 0.3 0.3>
    componentLinearVelScale=<0.3 0.3 0.3>
    gravityOverride {z=-980}
    physicsBodyDefinitions [{
      boundBone {boneName=dyn_hairStrand_S_1}
      boxExtents=<10 10 10>
      constraintSetup {angularLimitsMax=<0 10 15>, angularLimitsMin=<0 -5 -10>, angularTarget=<1 0 0>}
      localJointOffset=<1 0 0>
     }]
   }
   n12 = Anim Dynamics - Bone: dyn_hairStrand_T_1!(ComponentPose=@n11, Alpha=1, GravityOverride:Vector="0,0,-980") -> (Pose) {
    angularDampingOverride=1
    angularSpringConstant=1200
    bAngularSpring=true
    bOverrideAngularDamping=true
    bUseGravityOverride=true
    gravityScale=0.35
    simulationSpace=RootRelative
    alphaBoolBlend {blendOption=Linear}
    boundBone {boneName=dyn_hairStrand_T_1}
    componentAppliedLinearAccClamp=<5000 5000 5000>
    componentLinearAccScale=<0.3 0.3 0.3>
    componentLinearVelScale=<0.3 0.3 0.3>
    gravityOverride {z=-980}
    physicsBodyDefinitions [{
      boundBone {boneName=dyn_hairStrand_T_1}
      boxExtents=<10 10 10>
      constraintSetup {angularLimitsMax=<0 15 10>, angularLimitsMin=<0 0 -8>, angularTarget=<1 0 0>}
      localJointOffset=<1 0 0>
     }]
   }
   n13 = Anim Dynamics - Bone: dyn_hairStrand_V_1!(ComponentPose=@n12, Alpha=1, GravityOverride:Vector="0,0,-980") -> (Pose) {
    angularDampingOverride=1
    angularSpringConstant=1200
    bAngularSpring=true
    bOverrideAngularDamping=true
    bUseGravityOverride=true
    gravityScale=0.35
    simulationSpace=RootRelative
    alphaBoolBlend {blendOption=Linear}
    boundBone {boneName=dyn_hairStrand_V_1}
    componentAppliedLinearAccClamp=<5000 5000 5000>
    componentLinearAccScale=<0.3 0.3 0.3>
    componentLinearVelScale=<0.3 0.3 0.3>
    gravityOverride {z=-980}
    physicsBodyDefinitions [{
      boundBone {boneName=dyn_hairStrand_V_1}
      boxExtents=<10 10 10>
      constraintSetup {angularLimitsMax=<0 15 15>, angularLimitsMin=<0 0 -5>, angularTarget=<1 0 0>}
      localJointOffset=<1 0 0>
     }]
   }
   n14 = Anim Dynamics - Bone: dyn_hairStrand_A_1!(ComponentPose=@n13, Alpha=1, GravityOverride:Vector="0,0,-980") -> (Pose) {
    angularDampingOverride=1
    angularSpringConstant=1200
    bAngularSpring=true
    bOverrideAngularDamping=true
    bUseGravityOverride=true
    gravityScale=0.35
    simulationSpace=RootRelative
    alphaBoolBlend {blendOption=Linear}
    boundBone {boneName=dyn_hairStrand_A_1}
    componentAppliedLinearAccClamp=<5000 5000 5000>
    componentLinearAccScale=<0.3 0.3 0.3>
    componentLinearVelScale=<0.3 0.3 0.3>
    gravityOverride {z=-980}
    physicsBodyDefinitions [{
      boundBone {boneName=dyn_hairStrand_A_1}
      boxExtents=<10 10 10>
      constraintSetup {angularLimitsMax=<0 15 15>, angularLimitsMin=<0 -15 -15>, angularTarget=<1 0 0>}
      localJointOffset=<1 0 0>
     }]
   }
   n15 = Anim Dynamics - Bone: dyn_hairStrand_D_1!(ComponentPose=@n14, Alpha=1, GravityOverride:Vector="0,0,-980") -> (Pose) {
    angularDampingOverride=1
    angularSpringConstant=1200
    bAngularSpring=true
    bOverrideAngularDamping=true
    bUseGravityOverride=true
    gravityScale=0.35
    simulationSpace=RootRelative
    alphaBoolBlend {blendOption=Linear}
    boundBone {boneName=dyn_hairStrand_D_1}
    componentAppliedLinearAccClamp=<5000 5000 5000>
    componentLinearAccScale=<0.3 0.3 0.3>
    componentLinearVelScale=<0.3 0.3 0.3>
    gravityOverride {z=-980}
    physicsBodyDefinitions [{
      boundBone {boneName=dyn_hairStrand_D_1}
      boxExtents=<10 10 10>
      constraintSetup {angularLimitsMax=<0 15 15>, angularLimitsMin=<0 0 -15>, angularTarget=<1 0 0>}
      localJointOffset=<1 0 0>
     }]
   }
   n16 = Anim Dynamics - Bone: dyn_hairStrand_E_1!(ComponentPose=@n15, Alpha=1, GravityOverride:Vector="0,0,-980") -> (Pose) {
    angularDampingOverride=1
    angularSpringConstant=1200
    bAngularSpring=true
    bOverrideAngularDamping=true
    bUseGravityOverride=true
    gravityScale=0.35
    simulationSpace=RootRelative
    alphaBoolBlend {blendOption=Linear}
    boundBone {boneName=dyn_hairStrand_E_1}
    componentAppliedLinearAccClamp=<5000 5000 5000>
    componentLinearAccScale=<0.3 0.3 0.3>
    componentLinearVelScale=<0.3 0.3 0.3>
    gravityOverride {z=-980}
    physicsBodyDefinitions [{
      boundBone {boneName=dyn_hairStrand_E_1}
      boxExtents=<10 10 10>
      constraintSetup {angularLimitsMax=<0 15 15>, angularLimitsMin=<0 0 -15>, angularTarget=<1 0 0>}
      localJointOffset=<1 0 0>
     }]
   }
   n17 = Anim Dynamics - Bone: dyn_hairStrand_I_1!(ComponentPose=@n16, Alpha=1, GravityOverride:Vector="0,0,-980") -> (Pose) {
    angularDampingOverride=1
    angularSpringConstant=1200
    bAngularSpring=true
    bOverrideAngularDamping=true
    bUseGravityOverride=true
    gravityScale=0.35
    simulationSpace=RootRelative
    alphaBoolBlend {blendOption=Linear}
    boundBone {boneName=dyn_hairStrand_I_1}
    componentAppliedLinearAccClamp=<5000 5000 5000>
    componentLinearAccScale=<0.3 0.3 0.3>
    componentLinearVelScale=<0.3 0.3 0.3>
    gravityOverride {z=-980}
    physicsBodyDefinitions [{
      boundBone {boneName=dyn_hairStrand_I_1}
      boxExtents=<10 10 10>
      constraintSetup {angularLimitsMax=<0 15 15>, angularLimitsMin=<0 0 -15>, angularTarget=<1 0 0>}
      localJointOffset=<1 0 0>
     }]
   }
   n18 = Anim Dynamics - Bone: dyn_hairStrand_J_1!(ComponentPose=@n17, Alpha=1, GravityOverride:Vector="0,0,-980") -> (Pose) {
    angularDampingOverride=1
    angularSpringConstant=1200
    bAngularSpring=true
    bOverrideAngularDamping=true
    bUseGravityOverride=true
    gravityScale=0.35
    simulationSpace=RootRelative
    alphaBoolBlend {blendOption=Linear}
    boundBone {boneName=dyn_hairStrand_J_1}
    componentAppliedLinearAccClamp=<5000 5000 5000>
    componentLinearAccScale=<0.3 0.3 0.3>
    componentLinearVelScale=<0.3 0.3 0.3>
    gravityOverride {z=-980}
    physicsBodyDefinitions [{
      boundBone {boneName=dyn_hairStrand_J_1}
      boxExtents=<10 10 10>
      constraintSetup {angularLimitsMax=<0 15 15>, angularLimitsMin=<0 -15 -15>, angularTarget=<1 0 0>}
      localJointOffset=<1 0 0>
     }]
   }
   n19 = Anim Dynamics - Bone: dyn_hairStrand_K_1!(ComponentPose=@n18, Alpha=1, GravityOverride:Vector="0,0,-980") -> (Pose) {
    angularDampingOverride=1
    angularSpringConstant=1200
    bAngularSpring=true
    bOverrideAngularDamping=true
    bUseGravityOverride=true
    gravityScale=0.35
    simulationSpace=RootRelative
    alphaBoolBlend {blendOption=Linear}
    boundBone {boneName=dyn_hairStrand_K_1}
    componentAppliedLinearAccClamp=<5000 5000 5000>
    componentLinearAccScale=<0.3 0.3 0.3>
    componentLinearVelScale=<0.3 0.3 0.3>
    gravityOverride {z=-980}
    physicsBodyDefinitions [{
      boundBone {boneName=dyn_hairStrand_K_1}
      boxExtents=<10 10 10>
      constraintSetup {angularLimitsMax=<0 15 15>, angularLimitsMin=<0 -15 -15>, angularTarget=<1 0 0>}
      localJointOffset=<1 0 0>
     }]
   }
   n20 = Anim Dynamics - Bone: dyn_hairStrand_L_1!(ComponentPose=@n19, Alpha=1, GravityOverride:Vector="0,0,-980") -> (Pose) {
    angularDampingOverride=1
    angularSpringConstant=1200
    bAngularSpring=true
    bOverrideAngularDamping=true
    bUseGravityOverride=true
    gravityScale=0.35
    simulationSpace=RootRelative
    alphaBoolBlend {blendOption=Linear}
    boundBone {boneName=dyn_hairStrand_L_1}
    componentAppliedLinearAccClamp=<5000 5000 5000>
    componentLinearAccScale=<0.3 0.3 0.3>
    componentLinearVelScale=<0.3 0.3 0.3>
    gravityOverride {z=-980}
    physicsBodyDefinitions [{
      boundBone {boneName=dyn_hairStrand_L_1}
      boxExtents=<10 10 10>
      constraintSetup {angularLimitsMax=<0 15 15>, angularLimitsMin=<0 -15 -15>, angularTarget=<1 0 0>}
      localJointOffset=<1 0 0>
     }]
   }
   n21 = Anim Dynamics - Bone: dyn_hairStrand_W_1!(ComponentPose=@n20, Alpha=1, GravityOverride:Vector="0,0,-980") -> (Pose) {
    angularDampingOverride=1
    angularSpringConstant=1200
    bAngularSpring=true
    bOverrideAngularDamping=true
    bUseGravityOverride=true
    gravityScale=0.35
    simulationSpace=RootRelative
    alphaBoolBlend {blendOption=Linear}
    boundBone {boneName=dyn_hairStrand_W_1}
    componentAppliedLinearAccClamp=<5000 5000 5000>
    componentLinearAccScale=<0.3 0.3 0.3>
    componentLinearVelScale=<0.3 0.3 0.3>
    gravityOverride {z=-980}
    physicsBodyDefinitions [{
      boundBone {boneName=dyn_hairStrand_W_1}
      boxExtents=<10 10 10>
      constraintSetup {angularLimitsMax=<0 15 15>, angularLimitsMin=<0 -15 -15>, angularTarget=<1 0 0>}
      localJointOffset=<1 0 0>
     }]
   }
   n22 = Anim Dynamics - Bone: dyn_hairStrand_Y_1!(ComponentPose=@n21, Alpha=1, GravityOverride:Vector="0,0,-980") -> (Pose) {
    angularDampingOverride=1
    angularSpringConstant=1200
    bAngularSpring=true
    bOverrideAngularDamping=true
    bUseGravityOverride=true
    gravityScale=0.35
    simulationSpace=RootRelative
    alphaBoolBlend {blendOption=Linear}
    boundBone {boneName=dyn_hairStrand_Y_1}
    componentAppliedLinearAccClamp=<5000 5000 5000>
    componentLinearAccScale=<0.3 0.3 0.3>
    componentLinearVelScale=<0.3 0.3 0.3>
    gravityOverride {z=-980}
    physicsBodyDefinitions [{
      boundBone {boneName=dyn_hairStrand_Y_1}
      boxExtents=<10 10 10>
      constraintSetup {angularLimitsMax=<0 15 15>, angularLimitsMin=<0 -15 -15>, angularTarget=<1 0 0>}
      localJointOffset=<1 0 0>
     }]
   }
   n23 = Anim Dynamics - Bone: dyn_hairStrand_Z_1!(ComponentPose=@n22, Alpha=1, GravityOverride:Vector="0,0,-980") -> (Pose) {
    angularDampingOverride=1
    angularSpringConstant=1200
    bAngularSpring=true
    bOverrideAngularDamping=true
    bUseGravityOverride=true
    gravityScale=0.35
    simulationSpace=RootRelative
    alphaBoolBlend {blendOption=Linear}
    boundBone {boneName=dyn_hairStrand_Z_1}
    componentAppliedLinearAccClamp=<5000 5000 5000>
    componentLinearAccScale=<0.3 0.3 0.3>
    componentLinearVelScale=<0.3 0.3 0.3>
    gravityOverride {z=-980}
    physicsBodyDefinitions [{
      boundBone {boneName=dyn_hairStrand_Z_1}
      boxExtents=<10 10 10>
      constraintSetup {angularLimitsMax=<0 15 15>, angularLimitsMin=<0 -15 -15>, angularTarget=<1 0 0>}
      localJointOffset=<1 0 0>
     }]
   }
   n24 = Anim Dynamics - Bone: dyn_hairStrand_B_1!(ComponentPose=@n23, Alpha=1, GravityOverride:Vector="0,0,-980") -> (Pose) {
    angularDampingOverride=1
    angularSpringConstant=1200
    bAngularSpring=true
    bOverrideAngularDamping=true
    bUseGravityOverride=true
    gravityScale=0.35
    simulationSpace=RootRelative
    alphaBoolBlend {blendOption=Linear}
    boundBone {boneName=dyn_hairStrand_B_1}
    componentAppliedLinearAccClamp=<5000 5000 5000>
    componentLinearAccScale=<0.3 0.3 0.3>
    componentLinearVelScale=<0.3 0.3 0.3>
    gravityOverride {z=-980}
    physicsBodyDefinitions [{
      boundBone {boneName=dyn_hairStrand_B_1}
      boxExtents=<10 10 10>
      constraintSetup {angularLimitsMax=<0 10 15>, angularLimitsMin=<0 -30 -15>, angularTarget=<1 0 0>}
      localJointOffset=<1 0 0>
     }]
   }
   n25 = Anim Dynamics - Bone: dyn_hairStrand_C_1!(ComponentPose=@n24, Alpha=1, GravityOverride:Vector="0,0,-980") -> (Pose) {
    angularDampingOverride=1
    angularSpringConstant=1200
    bAngularSpring=true
    bOverrideAngularDamping=true
    bUseGravityOverride=true
    gravityScale=0.35
    simulationSpace=RootRelative
    alphaBoolBlend {blendOption=Linear}
    boundBone {boneName=dyn_hairStrand_C_1}
    componentAppliedLinearAccClamp=<5000 5000 5000>
    componentLinearAccScale=<0.3 0.3 0.3>
    componentLinearVelScale=<0.3 0.3 0.3>
    gravityOverride {z=-980}
    physicsBodyDefinitions [{
      boundBone {boneName=dyn_hairStrand_C_1}
      boxExtents=<10 10 10>
      constraintSetup {angularLimitsMax=<0 15 7>, angularLimitsMin=<0 0 -7>, angularTarget=<1 0 0>}
      localJointOffset=<1 0 0>
     }]
   }
   n26 = Anim Dynamics - Bone: dyn_hairStrand_F_1!(ComponentPose=@n25, Alpha=1, GravityOverride:Vector="0,0,-980") -> (Pose) {
    angularDampingOverride=1
    angularSpringConstant=1200
    bAngularSpring=true
    bOverrideAngularDamping=true
    bUseGravityOverride=true
    gravityScale=0.35
    simulationSpace=RootRelative
    alphaBoolBlend {blendOption=Linear}
    boundBone {boneName=dyn_hairStrand_F_1}
    componentAppliedLinearAccClamp=<5000 5000 5000>
    componentLinearAccScale=<0.3 0.3 0.3>
    componentLinearVelScale=<0.3 0.3 0.3>
    gravityOverride {z=-980}
    physicsBodyDefinitions [{
      boundBone {boneName=dyn_hairStrand_F_1}
      boxExtents=<10 10 10>
      constraintSetup {angularLimitsMax=<0 20 5>, angularLimitsMin=<0 -5 -5>, angularTarget=<1 0 0>}
      localJointOffset=<1 0 0>
     }]
   }
   n27 = Anim Dynamics - Bone: dyn_hairStrand_H_1!(ComponentPose=@n26, Alpha=1, GravityOverride:Vector="0,0,-980") -> (Pose) {
    angularDampingOverride=1
    angularSpringConstant=1200
    bAngularSpring=true
    bOverrideAngularDamping=true
    bUseGravityOverride=true
    gravityScale=0.35
    simulationSpace=RootRelative
    alphaBoolBlend {blendOption=Linear}
    boundBone {boneName=dyn_hairStrand_H_1}
    componentAppliedLinearAccClamp=<5000 5000 5000>
    componentLinearAccScale=<0.3 0.3 0.3>
    componentLinearVelScale=<0.3 0.3 0.3>
    gravityOverride {z=-980}
    physicsBodyDefinitions [{
      boundBone {boneName=dyn_hairStrand_H_1}
      boxExtents=<10 10 10>
      constraintSetup {angularLimitsMax=<0 15 15>, angularLimitsMin=<0 -15 -15>, angularTarget=<1 0 0>}
      localJointOffset=<1 0 0>
     }]
   }
   n28 = Anim Dynamics - Bone: dyn_hairStrand_U_1!(ComponentPose=@n27, Alpha=1, GravityOverride:Vector="0,0,-980") -> (Pose) {
    angularDampingOverride=1
    angularSpringConstant=1200
    bAngularSpring=true
    bOverrideAngularDamping=true
    bUseGravityOverride=true
    gravityScale=0.35
    simulationSpace=RootRelative
    alphaBoolBlend {blendOption=Linear}
    boundBone {boneName=dyn_hairStrand_U_1}
    componentAppliedLinearAccClamp=<5000 5000 5000>
    componentLinearAccScale=<0.3 0.3 0.3>
    componentLinearVelScale=<0.3 0.3 0.3>
    gravityOverride {z=-980}
    physicsBodyDefinitions [{
      boundBone {boneName=dyn_hairStrand_U_1}
      boxExtents=<10 10 10>
      constraintSetup {angularLimitsMax=<0 15 15>, angularLimitsMin=<0 -15 -15>, angularTarget=<1 0 0>}
      localJointOffset=<1 0 0>
     }]
   }
   n29 = Anim Dynamics - Bone: dyn_hairStrand_X_1!(ComponentPose=@n28, Alpha=1, GravityOverride:Vector="0,0,-980") -> (Pose) {
    angularDampingOverride=1
    angularSpringConstant=1200
    bAngularSpring=true
    bOverrideAngularDamping=true
    bUseGravityOverride=true
    gravityScale=0.35
    simulationSpace=RootRelative
    alphaBoolBlend {blendOption=Linear}
    boundBone {boneName=dyn_hairStrand_X_1}
    componentAppliedLinearAccClamp=<5000 5000 5000>
    componentLinearAccScale=<0.3 0.3 0.3>
    componentLinearVelScale=<0.3 0.3 0.3>
    gravityOverride {z=-980}
    physicsBodyDefinitions [{
      boundBone {boneName=dyn_hairStrand_X_1}
      boxExtents=<10 10 10>
      constraintSetup {angularLimitsMax=<0 15 15>, angularLimitsMin=<0 -15 -15>, angularTarget=<1 0 0>}
      localJointOffset=<1 0 0>
     }]
   }
   n30 = RigidBody!(
    ComponentPose=@n29
    Alpha
    OverrideWorldGravity:Vector="0,0,-980"
    SimSpaceSettings:Sim Space Settings Structure=(WorldAlpha=1,VelocityScaleZ=0.5,DampingAlpha=1,MaxLinearVelocity=400,MaxAngularVelocity=0.3,MaxLinearAcceleration=800,MaxAngularAcceleration=1,ExternalLinearDragV=(X=1,Y=1,Z=1),ExternalLinearVelocity=(X=0,Y=0,Z=0),ExternalAngularVelocity=(X=0,Y=0,Z=0))
   ) -> (Pose) {
    bClampLinearTranslationLimitToRefPose=true
    bOverrideWorldGravity=true
    overridePhysicsAsset=&/F_MED_PeonyBellow_FaceAcc_Physics
    simulationSpace=BaseBoneSpace
    alphaBoolBlend {blendOption=Linear}
    baseBoneRef {boneName=spine_05}
    componentAppliedLinearAccClamp=<3500 3500 3500>
    overrideWorldGravity {z=-980}
    simSpaceSettings {
     maxAngularAcceleration=1
     maxAngularVelocity=0.3
     maxLinearAcceleration=800
     maxLinearVelocity=400
     velocityScaleZ=0.5
     worldAlpha=1
     externalLinearDragV=<1 1 1>
    }
   }
   Save cached pose 'dynamicsCP'(Pose=Component To Local(ComponentPose=@n30)) {cacheName=dynamicsCP}
   n31 = "Transform (Modify) Bone - Bone: FACIAL_L_Ear"!(
    ComponentPose=Local To Component(LocalPose=Use cached pose 'controlRigCP'())
    bAlphaBoolEnabled:bool=True
    Translation="0,0,0"
    Rotation="0,0,0"
   ) -> (Pose) {
    alphaInputType=Bool
    rotationMode=BMM_Replace
    rotationSpace=BCS_ParentBoneSpace
    translationMode=BMM_Replace
    translationSpace=BCS_ParentBoneSpace
    alphaBoolBlend {blendOption=Linear}
    boneToModify {boneName=FACIAL_L_Ear}
   }
   n32 = "Transform (Modify) Bone - Bone: FACIAL_R_Ear"!(ComponentPose=@n31, bAlphaBoolEnabled:bool=True, Translation="0,0,0", Rotation="0,0,0") -> (Pose) {
    alphaInputType=Bool
    rotationMode=BMM_Replace
    rotationSpace=BCS_ParentBoneSpace
    translationMode=BMM_Replace
    translationSpace=BCS_ParentBoneSpace
    alphaBoolBlend {blendOption=Linear}
    boneToModify {boneName=FACIAL_R_Ear}
   }
   n33 = "Transform (Modify) Bone - Bone: base_neckCollar_l_1"!(ComponentPose=@n32, bAlphaBoolEnabled:bool=True, Translation="0,0,0", Rotation="0,0,0") -> (Pose) {
    alphaInputType=Bool
    rotationMode=BMM_Replace
    rotationSpace=BCS_ParentBoneSpace
    translationMode=BMM_Replace
    translationSpace=BCS_ParentBoneSpace
    alphaBoolBlend {blendOption=Linear}
    boneToModify {boneName=base_neckCollar_l_1}
   }
   n34 = "Transform (Modify) Bone - Bone: base_neckCollar_l_2"!(ComponentPose=@n33, bAlphaBoolEnabled:bool=True, Translation="0,0,0", Rotation="0,0,0") -> (Pose) {
    alphaInputType=Bool
    rotationMode=BMM_Replace
    rotationSpace=BCS_ParentBoneSpace
    translationMode=BMM_Replace
    translationSpace=BCS_ParentBoneSpace
    alphaBoolBlend {blendOption=Linear}
    boneToModify {boneName=base_neckCollar_l_2}
   }
   n35 = "Transform (Modify) Bone - Bone: base_neckCollar_l_3"!(ComponentPose=@n34, bAlphaBoolEnabled:bool=True, Translation="0,0,0", Rotation="0,0,0") -> (Pose) {
    alphaInputType=Bool
    rotationMode=BMM_Replace
    rotationSpace=BCS_ParentBoneSpace
    translationMode=BMM_Replace
    translationSpace=BCS_ParentBoneSpace
    alphaBoolBlend {blendOption=Linear}
    boneToModify {boneName=base_neckCollar_l_3}
   }
   n36 = "Transform (Modify) Bone - Bone: base_neckCollar_r_1"!(ComponentPose=@n35, bAlphaBoolEnabled:bool=True, Translation="0,0,0", Rotation="0,0,0") -> (Pose) {
    alphaInputType=Bool
    rotationMode=BMM_Replace
    rotationSpace=BCS_ParentBoneSpace
    translationMode=BMM_Replace
    translationSpace=BCS_ParentBoneSpace
    alphaBoolBlend {blendOption=Linear}
    boneToModify {boneName=base_neckCollar_r_1}
   }
   n37 = "Transform (Modify) Bone - Bone: base_neckCollar_r_2"!(ComponentPose=@n36, bAlphaBoolEnabled:bool=True, Translation="0,0,0", Rotation="0,0,0") -> (Pose) {
    alphaInputType=Bool
    rotationMode=BMM_Replace
    rotationSpace=BCS_ParentBoneSpace
    translationMode=BMM_Replace
    translationSpace=BCS_ParentBoneSpace
    alphaBoolBlend {blendOption=Linear}
    boneToModify {boneName=base_neckCollar_r_2}
   }
   n38 = "Transform (Modify) Bone - Bone: base_neckCollar_r_3"!(ComponentPose=@n37, bAlphaBoolEnabled:bool=True, Translation="0,0,0", Rotation="0,0,0") -> (Pose) {
    alphaInputType=Bool
    rotationMode=BMM_Replace
    rotationSpace=BCS_ParentBoneSpace
    translationMode=BMM_Replace
    translationSpace=BCS_ParentBoneSpace
    alphaBoolBlend {blendOption=Linear}
    boneToModify {boneName=base_neckCollar_r_3}
   }
   n39 = "Transform (Modify) Bone - Bone: base_neckCollar_c_1"!(ComponentPose=@n38, bAlphaBoolEnabled:bool=True, Translation="0,0,0", Rotation="0,0,0") -> (Pose) {
    alphaInputType=Bool
    rotationMode=BMM_Replace
    rotationSpace=BCS_ParentBoneSpace
    translationMode=BMM_Replace
    translationSpace=BCS_ParentBoneSpace
    alphaBoolBlend {blendOption=Linear}
    boneToModify {boneName=base_neckCollar_c_1}
   }
   Save cached pose 'transferedTransformHeadBodyCP'(Pose=Component To Local(ComponentPose=@n39)) {cacheName=transferedTransformHeadBodyCP}
   n40 = "Constraint - dyn_earring_l_1 (PP)"!(
    ComponentPose=Local To Component(LocalPose=Use cached pose 'transferedTransformHeadBodyCP'())
    Alpha=1
    ConstraintWeights_0=0.8525
    ConstraintWeights_1:float=0.1475
   ) -> (Pose) {
    alphaBoolBlend {blendOption=Linear}
    boneToModify {boneName=dyn_earring_l_1}
    constraintSetup [{targetBone {boneName=FACIAL_L_Ear}}, {targetBone {boneName=head}}]
    constraintWeights [1,1]
   }
   n41 = "Constraint - dyn_earring_r_1 (PP)"!(ComponentPose=@n40, Alpha=1, ConstraintWeights_0=0.8525, ConstraintWeights_1:float=0.1475) -> (Pose) {
    alphaBoolBlend {blendOption=Linear}
    boneToModify {boneName=dyn_earring_r_1}
    constraintSetup [{targetBone {boneName=FACIAL_R_Ear}}, {targetBone {boneName=head}}]
    constraintWeights [1,1]
   }
   n42 = "Constraint - coll_earring_ponytail (PPPPPPPP)"!(
    ComponentPose=@n41
    Alpha=1
    ConstraintWeights_0=0.2564
    ConstraintWeights_1:float=0.2528
    ConstraintWeights_2:float=0.1439
    ConstraintWeights_3:float=0.1355
    ConstraintWeights_4:float=0.1058
    ConstraintWeights_5:float=0.1015
    ConstraintWeights_6:float=0.0033
    ConstraintWeights_7:float=0.0008
   ) -> (Pose) {
    alphaBoolBlend {blendOption=Linear}
    boneToModify {boneName=coll_earring_ponytail}
    constraintSetup [
     {targetBone {boneName=spine_05}}
     {targetBone {boneName=head}}
     {targetBone {boneName=dyn_ponytail_1}}
     {targetBone {boneName=clavicle_l}}
     {targetBone {boneName=neck_01}}
     {targetBone {boneName=neck_02}}
     {targetBone {boneName=dyn_ponytail_2}}
     {targetBone {boneName=dyn_ponytail_3}}
    ]
    constraintWeights [1,1,1,1,1,1,1,1]
   }
   n43 = "Transform (Modify) Bone - Bone: base_ponytail_1"!(ComponentPose=@n42, bAlphaBoolEnabled:bool=True, Translation="0,0,0", Rotation="-15,0,0") -> (Pose) {
    alphaInputType=Bool
    rotationMode=BMM_Additive
    rotationSpace=BCS_BoneSpace
    translationMode=BMM_Additive
    translationSpace=BCS_BoneSpace
    alphaBoolBlend {blendInTime=2, blendOption=Sinusoidal, blendOutTime=2}
    boneToModify {boneName=base_ponytail_1}
    rotation {pitch=-15}
   }
   n44 = "Transform (Modify) Bone - Bone: base_earring_l_1"!(ComponentPose=@n43, Alpha=1, Translation="0,0,0", Rotation="0,18,0") -> (Pose) {
    rotationMode=BMM_Additive
    rotationSpace=BCS_BoneSpace
    scaleSpace=BCS_WorldSpace
    translationMode=BMM_Additive
    translationSpace=BCS_BoneSpace
    alphaBoolBlend {blendOption=Linear}
    alphaScaleBiasClamp {bInterpResult=true, interpSpeedDecreasing=2, interpSpeedIncreasing=2}
    boneToModify {boneName=base_earring_l_1}
    rotation {yaw=18}
   }
   n45 = "Transform (Modify) Bone - Bone: base_earring_r_1"!(ComponentPose=@n44, Alpha=1, Translation="0,0,0", Rotation="0,-22,-10") -> (Pose) {
    rotationMode=BMM_Additive
    rotationSpace=BCS_BoneSpace
    scaleSpace=BCS_WorldSpace
    translationMode=BMM_Additive
    translationSpace=BCS_BoneSpace
    alphaBoolBlend {blendOption=Linear}
    alphaScaleBiasClamp {bInterpResult=true, interpSpeedDecreasing=2, interpSpeedIncreasing=2}
    boneToModify {boneName=base_earring_r_1}
    rotation {roll=-10, yaw=-22}
   }
   Save cached pose 'presimFixesCP'(Pose=Component To Local(ComponentPose=@n45)) {cacheName=presimFixesCP}
  ]}
 EventGraph {nodes [
   Event BlueprintUpdateAnimation()
   > n1 = Sequence!(execute) -> (then_0, then_1, then_2:Exec)
   > n2 = Gravity Override State Function DEPRECATED(
    CharacterPartAnimInstance=Self-Reference()
    GameTimeInSeconds=GetGameTimeInSeconds()
    GravityOverrideParams=Get Gravity Override Params()
    RBAN Params=Get RBAN Params()
   )
   > Set GravityOverride(GravityOverride=@n2)
   > Set Sim Space Settings(Sim Space Settings:Sim Space Settings Structure=Break RBANSetting Struct!(RBANSettingStruct:"RBANSetting Struct Structure (by ref)"=@n2.RBAN SettingsOut) -> (
     ComponentLinearAccScale_28_#1:Vector=~
     ComponentLinearVelScale_31_#2:Vector=~
     SimSpaceSettings_34_#3:Sim Space Settings Structure
    ) {is=title K2Node_BreakStruct}.SimSpaceSettings_34_#3) -> (Output_Get:Sim Space Settings Structure)
   n3 = Sequence(execute=@n1.then_2)
   n4 = Get BodyMesh()
   > Is Valid(InputObject=@n4)
   n5 = GetSocketBoneName!(self:Skinned Mesh Component Object Reference=@n4, InSocketName:Name=upperarm_r) -> (ReturnValue:Name=None)
   > n6 = Set isBodyAttached(isBodyAttached:bool="Not Equal Exactly (String)"(A="To String (Name)"(InName=@n5), B=None)) -> (Output_Get:bool)
   > Branch(Condition=@n6)
   n7 = Get BodyMesh()
   n8 = BreakTransform(InTransform=GetSocketTransform(self=@n7, InSocketName=base_neckCollar_c_1, TransformSpace=RTS_ParentBoneSpace))
   > Set base_neckCollar_c_1_location(base_neckCollar_c_1_location:Vector=@n8) -> (Output_Get:Vector)
   > Set base_neckCollar_c_1_rotation(base_neckCollar_c_1_rotation:Rotator=@n8.Rotation) -> (Output_Get:Rotator)
   n9 = BreakTransform(InTransform=GetSocketTransform(self=@n7, InSocketName=base_neckCollar_l_1, TransformSpace=RTS_ParentBoneSpace))
   > Set base_neckCollar_l_1_location(base_neckCollar_l_1_location:Vector=@n9) -> (Output_Get:Vector)
   > Set base_neckCollar_l_1_rotation(base_neckCollar_l_1_rotation:Rotator=@n9.Rotation) -> (Output_Get:Rotator)
   n10 = BreakTransform(InTransform=GetSocketTransform(self=@n7, InSocketName=base_neckCollar_l_2, TransformSpace=RTS_ParentBoneSpace))
   > Set base_neckCollar_l_2_location(base_neckCollar_l_2_location:Vector=@n10) -> (Output_Get:Vector)
   > Set base_neckCollar_l_2_rotation(base_neckCollar_l_2_rotation:Rotator=@n10.Rotation) -> (Output_Get:Rotator)
   n11 = BreakTransform(InTransform=GetSocketTransform(self=@n7, InSocketName=base_neckCollar_l_3, TransformSpace=RTS_ParentBoneSpace))
   > Set base_neckCollar_l_3_location(base_neckCollar_l_3_location:Vector=@n11) -> (Output_Get:Vector)
   > Set base_neckCollar_l_3_rotation(base_neckCollar_l_3_rotation:Rotator=@n11.Rotation) -> (Output_Get:Rotator)
   n12 = BreakTransform(InTransform=GetSocketTransform(self=@n7, InSocketName=base_neckCollar_r_1, TransformSpace=RTS_ParentBoneSpace))
   > Set base_neckCollar_r_1_location(base_neckCollar_r_1_location:Vector=@n12) -> (Output_Get:Vector)
   > Set base_neckCollar_r_1_rotation(base_neckCollar_r_1_rotation:Rotator=@n12.Rotation) -> (Output_Get:Rotator)
   n13 = BreakTransform(InTransform=GetSocketTransform(self=@n7, InSocketName=base_neckCollar_r_2, TransformSpace=RTS_ParentBoneSpace))
   > Set base_neckCollar_r_2_location(base_neckCollar_r_2_location:Vector=@n13) -> (Output_Get:Vector)
   > Set base_neckCollar_r_2_rotation(base_neckCollar_r_2_rotation:Rotator=@n13.Rotation) -> (Output_Get:Rotator)
   n14 = BreakTransform(InTransform=GetSocketTransform(self=@n7, InSocketName=base_neckCollar_r_3, TransformSpace=RTS_ParentBoneSpace))
   > Set base_neckCollar_r_3_location(base_neckCollar_r_3_location:Vector=@n14) -> (Output_Get:Vector)
   > Set base_neckCollar_r_3_rotation(base_neckCollar_r_3_rotation:Rotator=@n14.Rotation) -> (Output_Get:Rotator)
   n15 = Branch(
    execute=@n3.then_1
    Condition=OR Boolean(A=Get bIsSkydiveFloating(), B=Get isUsingHangGlidingInParachute() -> (isUsingHangGlidingInParachute:bool))
   )
   > Set adjustCoatCollar weight(adjustCoatCollar weight:double=1) -> (Output_Get:double)
   Set adjustCoatCollar weight(execute=@n15.else, adjustCoatCollar weight=0)
   Set backHairStrands weight(
    execute=@n3.then_2
    backHairStrands weight:double=Remap Float(Input=GetCurveValue(CurveName=headRz_up_out), OldRange=(X=0,Y=1), NewRange=(X=1,Y=0), Clamp=true) -> (Output=~)
   ) -> (Output_Get:double)
   n16 = Get HeadMesh() -> (HeadMesh:Skeletal Mesh Component Budgeted Object Reference)
   Is Valid(exec=@n1.then_1, InputObject=@n16)
   n17 = GetSocketBoneName(self=@n16, InSocketName=FACIAL_L_Ear) -> (ReturnValue=None)
   > n18 = Set isUsingFacialBones(isUsingFacialBones:bool="Not Equal Exactly (String)"(A="To String (Name)"(InName=@n17), B=None)) -> (Output_Get:bool)
   > Branch(Condition=@n18)
   n19 = Get HeadMesh()
   n20 = BreakTransform(InTransform=GetSocketTransform(self=@n19, InSocketName=FACIAL_L_Ear, TransformSpace=RTS_ParentBoneSpace))
   > Set FACIAL_L_Ear_translation(FACIAL_L_Ear_translation:Vector=@n20) -> (Output_Get:Vector)
   > Set FACIAL_L_Ear_rotation(FACIAL_L_Ear_rotation:Rotator=@n20.Rotation) -> (Output_Get:Rotator)
   n21 = BreakTransform(InTransform=GetSocketTransform(self=@n19, InSocketName=FACIAL_R_Ear, TransformSpace=RTS_ParentBoneSpace))
   > Set FACIAL_R_Ear_translation(FACIAL_R_Ear_translation:Vector=@n21) -> (Output_Get:Vector)
   > Set FACIAL_R_Ear_rotation(FACIAL_R_Ear_rotation:Rotator=@n21.Rotation) -> (Output_Get:Rotator)
   Event OnParachuteOpened()
   > Switch on EGliderType(Selection=Get PlayerGliderType() -> (PlayerGliderType=HangGlider))
   > Set isUsingHangGlidingInParachute(isUsingHangGlidingInParachute:bool=true) -> (Output_Get:bool)
   Event OnParachuteClosed()
   > Set isUsingHangGlidingInParachute(isUsingHangGlidingInParachute=false)
   n22 = Event OnCharacterCustomizationLoaded()
   > Set HeadMesh(HeadMesh:Skeletal Mesh Component Budgeted Object Reference=Get Array Item with EFort Custom Part Type(Array=@n22.CustomMeshParts, PartType=Head)) -> (Output_Get:Skeletal Mesh Component Budgeted Object Reference)
   > Set BodyMesh(BodyMesh=Get Array Item with EFort Custom Part Type(Array=@n22.CustomMeshParts))
  ]}
]
variables [
 adjustCoatCollar weight {type=number}
 backHairStrands weight {type=number}
 base_neckCollar_c_1_location {type=Vector}
 base_neckCollar_c_1_rotation {type=Rotator}
 base_neckCollar_l_1_location {type=Vector}
 base_neckCollar_l_1_rotation {type=Rotator}
 base_neckCollar_l_2_location {type=Vector}
 base_neckCollar_l_2_rotation {type=Rotator}
 base_neckCollar_l_3_location {type=Vector}
 base_neckCollar_l_3_rotation {type=Rotator}
 base_neckCollar_r_1_location {type=Vector}
 base_neckCollar_r_1_rotation {type=Rotator}
 base_neckCollar_r_2_location {type=Vector}
 base_neckCollar_r_2_rotation {type=Rotator}
 base_neckCollar_r_3_location {type=Vector}
 base_neckCollar_r_3_rotation {type=Rotator}
 BodyMesh {type=SkeletalMeshComponent}
 FACIAL_L_Ear_rotation {type=Rotator}
 FACIAL_L_Ear_translation {type=Vector}
 FACIAL_R_Ear_rotation {type=Rotator}
 FACIAL_R_Ear_translation {type=Vector}
 Gravity Override Params {type=GravityOverrideParamsStruct}
 GravityOverride {type=Vector}
 HeadMesh {type=SkeletalMeshComponentBudgeted}
 isBodyAttached {type=boolean}
 isUsingFacialBones {type=boolean}
 isUsingHangGlidingInParachute {type=boolean}
 RBAN Params {
  type=RBANParamsStruct
  default {
   dBNO {simSpaceSettings {
     maxAngularAcceleration=1
     maxAngularVelocity=0.3
     maxLinearAcceleration=1200
     maxLinearVelocity=400
     velocityScaleZ=0.5
     externalLinearDragV=<1 1 1>
    }}
   dBNO_Carried {simSpaceSettings {
     maxAngularAcceleration=1
     maxAngularVelocity=0.3
     maxLinearAcceleration=1200
     maxLinearVelocity=400
     velocityScaleZ=0.5
     externalLinearDragV=<1 1 1>
    }}
   emoteOrMelee {simSpaceSettings {
     maxAngularAcceleration=1
     maxAngularVelocity=0.3
     maxLinearAcceleration=1200
     maxLinearVelocity=400
     velocityScaleZ=0.5
     externalLinearDragV=<1 1 1>
    }}
   front_End {simSpaceSettings {
     maxAngularAcceleration=1
     maxAngularVelocity=0.3
     maxLinearAcceleration=1200
     maxLinearVelocity=400
     velocityScaleZ=0.5
     externalLinearDragV=<1 1 1>
    }}
   onGround_Crouch_Moving {simSpaceSettings {
     maxAngularAcceleration=1
     maxAngularVelocity=0.3
     maxLinearAcceleration=1200
     maxLinearVelocity=400
     velocityScaleZ=0.5
     externalLinearDragV=<1 1 1>
    }}
   onGround_Falling {simSpaceSettings {
     maxAngularAcceleration=1
     maxAngularVelocity=0.3
     maxLinearAcceleration=1200
     maxLinearVelocity=400
     velocityScaleZ=0.5
     externalLinearDragV=<1 1 1>
    }}
   onGround_Moving {
    componentLinearAccScale=<0 0 0>
    componentLinearVelScale=<0 0 0>
    simSpaceSettings {
     maxAngularAcceleration=1
     maxAngularVelocity=0.3
     maxLinearAcceleration=1600
     maxLinearVelocity=400
     velocityScaleZ=0.7
     externalLinearDragV=<1 1 1>
    }
   }
   onGround_Standing {
    componentLinearAccScale=<0 0 0>
    componentLinearVelScale=<0 0 0>
    simSpaceSettings {
     maxAngularAcceleration=1
     maxAngularVelocity=0.3
     maxLinearAcceleration=800
     maxLinearVelocity=400
     velocityScaleZ=0.5
     externalLinearDragV=<1 1 1>
    }
   }
   rail_Grinding {simSpaceSettings {
     maxAngularAcceleration=1
     maxAngularVelocity=0.3
     maxLinearAcceleration=1200
     maxLinearVelocity=400
     velocityScaleZ=0.5
     worldAlpha=1
     externalLinearDragV=<1 1 1>
    }}
   riding_Moving {simSpaceSettings {
     maxAngularAcceleration=1
     maxAngularVelocity=0.3
     maxLinearAcceleration=1200
     maxLinearVelocity=400
     velocityScaleZ=0.5
     externalLinearDragV=<1 1 1>
    }}
   riding_Moving_Targeting {simSpaceSettings {
     maxAngularAcceleration=1
     maxAngularVelocity=0.3
     maxLinearAcceleration=1200
     maxLinearVelocity=400
     velocityScaleZ=0.5
     externalLinearDragV=<1 1 1>
    }}
   riding_Standing {simSpaceSettings {
     maxAngularAcceleration=1
     maxAngularVelocity=0.3
     maxLinearAcceleration=1200
     maxLinearVelocity=400
     velocityScaleZ=0.5
     externalLinearDragV=<1 1 1>
    }}
   skydive_DivingDown {simSpaceSettings {
     maxAngularAcceleration=1
     maxAngularVelocity=0.3
     maxLinearAcceleration=1200
     maxLinearVelocity=400
     velocityScaleZ=0.5
     externalLinearDragV=<1 1 1>
    }}
   skydive_FloatingDown {
    componentLinearAccScale=<0 0.5 0>
    componentLinearVelScale=<0 0.5 0>
    simSpaceSettings {
     maxAngularAcceleration=1
     maxAngularVelocity=0.3
     maxLinearAcceleration=1200
     maxLinearVelocity=400
     velocityScaleZ=0.5
     externalLinearDragV=<1 1 1>
    }
   }
   skydive_GoingUp {simSpaceSettings {
     maxAngularAcceleration=1
     maxAngularVelocity=0.3
     maxLinearAcceleration=1200
     maxLinearVelocity=400
     velocityScaleZ=0.5
     externalLinearDragV=<1 1 1>
    }}
   skydive_Parachuting {
    componentLinearAccScale=<0 0 0.3>
    componentLinearVelScale=<0 0 0.3>
    simSpaceSettings {
     maxAngularAcceleration=1
     maxAngularVelocity=0.3
     maxLinearAcceleration=1200
     maxLinearVelocity=400
     velocityScaleZ=0.5
     externalLinearDragV=<1 1 1>
    }
   }
   skydive_Surfing {simSpaceSettings {
     maxAngularAcceleration=1
     maxAngularVelocity=0.3
     maxLinearAcceleration=1200
     maxLinearVelocity=400
     velocityScaleZ=0.5
     externalLinearDragV=<1 1 1>
    }}
   swimming {simSpaceSettings {
     maxAngularAcceleration=1
     maxAngularVelocity=0.3
     maxLinearAcceleration=1200
     maxLinearVelocity=400
     velocityScaleZ=0.5
     externalLinearDragV=<1 1 1>
    }}
   tactical_Sprint {simSpaceSettings {
     maxAngularAcceleration=1
     maxAngularVelocity=0.3
     maxLinearAcceleration=1200
     maxLinearVelocity=400
     velocityScaleZ=0.5
     externalLinearDragV=<1 1 1>
    }}
   ziplining {simSpaceSettings {
     maxAngularAcceleration=1
     maxAngularVelocity=0.3
     maxLinearAcceleration=1200
     maxLinearVelocity=400
     velocityScaleZ=0.5
     externalLinearDragV=<1 1 1>
    }}
  }
 }
 Sim Space Settings {
  type=SimSpaceSettings
  default {
   dampingAlpha=1
   maxAngularAcceleration=10000
   maxAngularVelocity=10000
   maxLinearAcceleration=10000
   maxLinearVelocity=10000
   velocityScaleZ=1
   worldAlpha=0
   externalLinearDragV=<0 0 0>
  }
 }
]
